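# Material function for the RadioPaca gadget prop that draws an animated LED spectrum bar graph, outputting Color, Mask and MaskedColor.
# UVs are quantized into bands and segments, with smoothstepped cell fractions forming the LED grid.
# Time-scrolled noise and a sound-reactivity path, switching between squared channels by horizontal position, supply bar levels that a height comparison thresholds.
# Three-color blends tint the LEDs, with a separate bars-only color path.
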MaterialFunction /BRCosmetics/Animation/Props/Gadgets_RadioPaca/Materials/MF_2DLEDSpectrum
graph {expressions {
  MaterialExpressionAdd_0 {
   class=MaterialExpressionAdd
   inputs [
    A {from=MaterialExpressionComponentMask_0}
    B {from=MaterialExpressionMultiply_7}
   ]
   properties {constB=3}
  }
  MaterialExpressionAdd_1 {
   class=MaterialExpressionAdd
   inputs [
    A {from=MaterialExpressionMultiply_14}
    B {from=MaterialExpressionComponentMask_7}
   ]
  }
  MaterialExpressionAdd_6 {
   class=MaterialExpressionAdd
   inputs [
    A {from=MaterialExpressionMultiply_21}
    B {from=MaterialExpressionMultiply_20}
   ]
  }
  MaterialExpressionAppendVector_0 {
   class=MaterialExpressionAppendVector
   inputs [
    A {from=MaterialExpressionDivide_1, output=""}
    B {from=MaterialExpressionDivide_0, output=""}
   ]
  }
  MaterialExpressionAppendVector_1 {
   class=MaterialExpressionAppendVector
   inputs [
    A {from=MaterialExpressionAdd_0, output=""}
    B {from=MaterialExpressionConstant_9, output=""}
   ]
  }
  MaterialExpressionAppendVector_2 {
   class=MaterialExpressionAppendVector
   inputs [
    A {from=MaterialExpressionNamedRerouteUsage_14, output=""}
    B {from=MaterialExpressionNamedRerouteUsage_15, output=""}
   ]
  }
  MaterialExpressionAppendVector_5 {
   class=MaterialExpressionAppendVector
   inputs [
    A {from=MaterialExpressionComponentMask_18, output=""}
    B {from=MaterialExpressionConstant_15, output=""}
   ]
  }
  MaterialExpressionComponentMask_0 {class=MaterialExpressionComponentMask, inputs [None {from=MaterialExpressionNamedRerouteUsage_5}], properties {r=true}}
  MaterialExpressionComponentMask_1 {class=MaterialExpressionComponentMask, inputs [None {from=MaterialExpressionNamedRerouteUsage_26}], properties {g=true}}
  MaterialExpressionComponentMask_18 {class=MaterialExpressionComponentMask, inputs [None {from=MaterialExpressionNamedRerouteUsage_52}], properties {r=true}}
  MaterialExpressionComponentMask_23 {class=MaterialExpressionComponentMask, inputs [None {from=MaterialExpressionNamedRerouteUsage_34}], properties {g=true}}
  MaterialExpressionComponentMask_7 {class=MaterialExpressionComponentMask, inputs [None {from=MaterialExpressionNamedRerouteUsage_0}], properties {g=true}}
  MaterialExpressionComponentMask_8 {class=MaterialExpressionComponentMask, inputs [None {from=MaterialExpressionNamedRerouteUsage_49}], properties {g=true}}
  MaterialExpressionConstant2Vector_2 {class=MaterialExpressionConstant2Vector, properties {g=30, r=30}}
  MaterialExpressionConstant3Vector_0 {class=MaterialExpressionConstant3Vector, properties {constant=<rgba 0.016 0.1878 0.227 1>}}
  MaterialExpressionConstant3Vector_1 {class=MaterialExpressionConstant3Vector, properties {constant=<rgba 0 0.2918 0.5841 1>}}
  MaterialExpressionConstant3Vector_2 {class=MaterialExpressionConstant3Vector, properties {constant=<rgba 0.1746 0.5457 0.6514 1>}}
  MaterialExpressionConstant_11 {class=MaterialExpressionConstant, properties {r=4}}
  MaterialExpressionConstant_14 {class=MaterialExpressionConstant, properties {r=2}}
  MaterialExpressionConstant_15 {class=MaterialExpressionConstant, properties {desc=a}}
  MaterialExpressionConstant_16 {class=MaterialExpressionConstant, properties {r=1}}
  MaterialExpressionConstant_17 {class=MaterialExpressionConstant}
  MaterialExpressionConstant_18 {class=MaterialExpressionConstant}
  MaterialExpressionConstant_5 {class=MaterialExpressionConstant, properties {r=0.4}}
  MaterialExpressionConstant_6 {class=MaterialExpressionConstant, properties {r=0.35}}
  MaterialExpressionConstant_9 {class=MaterialExpressionConstant, properties {desc=a}}
  MaterialExpressionConvert_0 {
   class=MaterialExpressionConvert
   inputs [None {from=MaterialExpressionNamedRerouteUsage_25}]
   properties {
    nodeName=Break Float2
    convertInputs [{
      type=Vector2
      expressionInput {
       expression=&self:MaterialExpressionNamedRerouteUsage_25
       inputName=None
       mask=0
       maskA=0
       maskB=0
       maskG=0
       maskR=0
       outputIndex=0
      }
     }]
    convertOutputs [{type=Scalar}, {type=Scalar}]
   }
  }
  MaterialExpressionConvert_1 {
   class=MaterialExpressionConvert
   inputs [None {from=MaterialExpressionFunctionInput_2}]
   properties {
    nodeName=Break Float2
    convertInputs [{
      type=Vector2
      expressionInput {
       expression=&self:MaterialExpressionFunctionInput_2
       inputName=None
       mask=0
       maskA=0
       maskB=0
       maskG=0
       maskR=0
       outputIndex=0
      }
     }]
    convertOutputs [{type=Scalar}, {type=Scalar}]
   }
  }
  MaterialExpressionDivide_0 {
   class=MaterialExpressionDivide
   inputs [
    A {from=MaterialExpressionFloor_1}
    B {from=MaterialExpressionNamedRerouteUsage_3}
   ]
  }
  MaterialExpressionDivide_1 {
   class=MaterialExpressionDivide
   inputs [
    A {from=MaterialExpressionFloor_0}
    B {from=MaterialExpressionNamedRerouteUsage_4}
   ]
  }
  MaterialExpressionDivide_9 {
   class=MaterialExpressionDivide
   inputs [
    A {from=MaterialExpressionComponentMask_7}
    B {from=MaterialExpressionNamedRerouteUsage_46}
   ]
  }
  MaterialExpressionFloor_0 {class=MaterialExpressionFloor, inputs [None {from=MaterialExpressionMultiply_0, output=""}], properties {}}
  MaterialExpressionFloor_1 {class=MaterialExpressionFloor, inputs [None {from=MaterialExpressionMultiply_1, output=""}], properties {}}
  MaterialExpressionFrac_0 {class=MaterialExpressionFrac, inputs [None {from=MaterialExpressionMultiply_2, output=""}], properties {}}
  MaterialExpressionFunctionInput_0 {
   class=MaterialExpressionFunctionInput
   inputs [Preview {from=MaterialExpressionPanner_0, output=""}]
   properties {bUsePreviewValueAsDefault=true, inputName=In_UV, inputType=FunctionInput_Vector2}
  }
  MaterialExpressionFunctionInput_10 {
   class=MaterialExpressionFunctionInput
   inputs [Preview {from="", output=""}]
   properties {bUsePreviewValueAsDefault=true, inputName=In_TimeStamp, inputType=FunctionInput_Scalar, sortPriority=10, previewValue=<0.5 0 0 1>}
  }
  MaterialExpressionFunctionInput_12 {class=MaterialExpressionFunctionInput, inputs [Preview {from="", output=""}], properties {inputName=In_JustBars}}
  MaterialExpressionFunctionInput_2 {
   class=MaterialExpressionFunctionInput
   inputs [Preview {from=MaterialExpressionConstant2Vector_2, output=""}]
   properties {inputName=In_Bands, inputType=FunctionInput_Vector2, sortPriority=2}
  }
  MaterialExpressionFunctionInput_4 {
   class=MaterialExpressionFunctionInput
   inputs [Preview {from=MaterialExpressionConstant3Vector_0, output=""}]
   properties {inputName=In_LowColor, sortPriority=4}
  }
  MaterialExpressionFunctionInput_5 {
   class=MaterialExpressionFunctionInput
   inputs [Preview {from=MaterialExpressionConstant3Vector_1, output=""}]
   properties {inputName=In_MidColor, sortPriority=5}
  }
  MaterialExpressionFunctionInput_6 {
   class=MaterialExpressionFunctionInput
   inputs [Preview {from=MaterialExpressionConstant3Vector_2, output=""}]
   properties {inputName=In_HighColor, sortPriority=6}
  }
  MaterialExpressionFunctionInput_8 {
   class=MaterialExpressionFunctionInput
   inputs [Preview {from="", output=""}]
   properties {bUsePreviewValueAsDefault=true, inputName=In_Spectrum, inputType=FunctionInput_Scalar, sortPriority=8, previewValue=<0.5 0 0 1>}
  }
  MaterialExpressionFunctionInput_9 {
   class=MaterialExpressionFunctionInput
   inputs [Preview {from="", output=""}]
   properties {bUsePreviewValueAsDefault=true, inputName=In_Amplitude, inputType=FunctionInput_Scalar, sortPriority=9, previewValue=<0.5 0 0 1>}
  }
  MaterialExpressionFunctionOutput_0 {
   class=MaterialExpressionFunctionOutput
   inputs [None {from=MaterialExpressionNamedRerouteUsage_56, output=""}]
   properties {outputName=Color}
  }
  MaterialExpressionFunctionOutput_1 {
   class=MaterialExpressionFunctionOutput
   inputs [None {from=MaterialExpressionNamedRerouteDeclaration_23, output=""}]
   properties {outputName=Mask, sortPriority=1}
  }
  MaterialExpressionFunctionOutput_2 {
   class=MaterialExpressionFunctionOutput
   inputs [None {from=MaterialExpressionMultiply_11, output=""}]
   properties {outputName=MaskedColor, sortPriority=2}
  }
  MaterialExpressionIf_1 {
   class=MaterialExpressionIf
   inputs [
    A {from=MaterialExpressionComponentMask_8, output=""}
    B {from=MaterialExpressionNamedRerouteUsage_50, output=""}
    "A > B" {from=MaterialExpressionConstant_17, output=""}
    "A == B" {from=MaterialExpressionConstant_16, output=""}
    "A < B" {from=MaterialExpressionConstant_16, output=""}
   ]
   properties {}
  }
  MaterialExpressionLinearInterpolate_0 {
   class=MaterialExpressionLinearInterpolate
   inputs [
    A {from=MaterialExpressionAdd_1}
    B {from=MaterialExpressionMultiply_12}
    Alpha {from=MaterialExpressionSaturate_5}
   ]
   properties {constA=0.1}
  }
  MaterialExpressionLinearInterpolate_1 {
   class=MaterialExpressionLinearInterpolate
   inputs [
    A {from=MaterialExpressionMultiply_15}
    B {from=MaterialExpressionMultiply_13}
    Alpha {from=MaterialExpressionNamedRerouteUsage_29}
   ]
  }
  MaterialExpressionLinearInterpolate_2 {
   class=MaterialExpressionLinearInterpolate
   inputs [
    A {from=MaterialExpressionMultiply_17}
    B {from=MaterialExpressionMultiply_8}
    Alpha {from=""}
   ]
  }
  MaterialExpressionLinearInterpolate_3 {
   class=MaterialExpressionLinearInterpolate
   inputs [
    A {from=MaterialExpressionConstant_18}
    B {from=MaterialExpressionMultiply_13}
    Alpha {from=MaterialExpressionNamedRerouteUsage_53}
   ]
  }
  MaterialExpressionLinearInterpolate_5 {
   class=MaterialExpressionLinearInterpolate
   inputs [
    A {from=MaterialExpressionNamedRerouteUsage_41}
    B {from=MaterialExpressionNamedRerouteUsage_57}
    Alpha {from=MaterialExpressionNamedRerouteUsage_59}
   ]
  }
  MaterialExpressionMaterialFunctionCall_0 {
   class=MaterialExpressionMaterialFunctionCall
   inputs [
    A {from=MaterialExpressionNamedRerouteUsage_7}
    B {from=MaterialExpressionNamedRerouteUsage_8}
    C {from=MaterialExpressionNamedRerouteUsage_9}
    Alpha {from=MaterialExpressionSquareRoot_0}
   ]
   properties {materialFunction=&3ColorBlend}
  }
  MaterialExpressionMaterialFunctionCall_1 {
   class=MaterialExpressionMaterialFunctionCall
   inputs [Float2 {from=MaterialExpressionNamedRerouteUsage_16}]
   properties {materialFunction=&BreakOutFloat2Components}
  }
  MaterialExpressionMaterialFunctionCall_2 {
   class=MaterialExpressionMaterialFunctionCall
   inputs [Multiplier {from=MaterialExpressionConstant_11}]
   properties {materialFunction=&Pi}
  }
  MaterialExpressionMaterialFunctionCall_5 {
   class=MaterialExpressionMaterialFunctionCall
   inputs [
    A {from=MaterialExpressionNamedRerouteUsage_37}
    B {from=MaterialExpressionNamedRerouteUsage_45}
    C {from=MaterialExpressionNamedRerouteUsage_47}
    Alpha {from=MaterialExpressionNamedRerouteUsage_44}
   ]
   properties {materialFunction=&3ColorBlend}
  }
  MaterialExpressionMaterialFunctionCall_6 {
   class=MaterialExpressionMaterialFunctionCall
   inputs [
    UVs {from=MaterialExpressionNamedRerouteUsage_19}
    Texture Scale {from=""}
   ]
   properties {materialFunction=&ScaleUVsByCenter}
  }
  MaterialExpressionMaterialFunctionCall_7 {
   class=MaterialExpressionMaterialFunctionCall
   inputs [Float4 {from=MaterialExpressionPower_0}]
   properties {materialFunction=&BreakOutFloat4Components}
  }
  MaterialExpressionMultiply_0 {
   class=MaterialExpressionMultiply
   inputs [
    A {from=MaterialExpressionNamedRerouteUsage_1}
    B {from=MaterialExpressionConvert_0}
   ]
  }
  MaterialExpressionMultiply_1 {
   class=MaterialExpressionMultiply
   inputs [
    A {from=MaterialExpressionConvert_0}
    B {from=MaterialExpressionNamedRerouteUsage_2}
   ]
  }
  MaterialExpressionMultiply_11 {
   class=MaterialExpressionMultiply
   inputs [
    A {from=MaterialExpressionLinearInterpolate_5}
    B {from=MaterialExpressionNamedRerouteUsage_40}
   ]
  }
  MaterialExpressionMultiply_12 {
   class=MaterialExpressionMultiply
   inputs [
    A {from=MaterialExpressionSaturate_6}
    B {from=MaterialExpressionIf_1}
   ]
  }
  MaterialExpressionMultiply_13 {
   class=MaterialExpressionMultiply
   inputs [
    A {from=MaterialExpressionNamedRerouteUsage_51}
    B {from=MaterialExpressionMaterialFunctionCall_5, output=Result}
   ]
  }
  MaterialExpressionMultiply_14 {
   class=MaterialExpressionMultiply
   inputs [
    A {from=MaterialExpressionMultiply_6}
    B {from=""}
   ]
   properties {constB=0.1}
  }
  MaterialExpressionMultiply_15 {
   class=MaterialExpressionMultiply
   inputs [
    A {from=MaterialExpressionMultiply_4}
    B {from=""}
   ]
   properties {constB=0.1}
  }
  MaterialExpressionMultiply_17 {
   class=MaterialExpressionMultiply
   inputs [
    A {from=MaterialExpressionSwitch_1}
    B {from=MaterialExpressionAdd_6}
   ]
  }
  MaterialExpressionMultiply_2 {
   class=MaterialExpressionMultiply
   inputs [
    A {from=MaterialExpressionSubtract_0}
    B {from=MaterialExpressionAppendVector_2}
   ]
  }
  MaterialExpressionMultiply_20 {
   class=MaterialExpressionMultiply
   inputs [
    A {from=MaterialExpressionNamedRerouteUsage_60}
    B {from=""}
   ]
   properties {constB=0.3}
  }
  MaterialExpressionMultiply_21 {
   class=MaterialExpressionMultiply
   inputs [
    A {from=MaterialExpressionNamedRerouteUsage_6}
    B {from=""}
   ]
   properties {constB=1.4}
  }
  MaterialExpressionMultiply_3 {
   class=MaterialExpressionMultiply
   inputs [
    A {from=MaterialExpressionSmoothStep_0}
    B {from=MaterialExpressionSmoothStep_1}
   ]
  }
  MaterialExpressionMultiply_4 {
   class=MaterialExpressionMultiply
   inputs [
    A {from=MaterialExpressionNamedRerouteUsage_17}
    B {from=MaterialExpressionNamedRerouteUsage_18}
   ]
  }
  MaterialExpressionMultiply_5 {
   class=MaterialExpressionMultiply
   inputs [
    A {from=MaterialExpressionAppendVector_5}
    B {from=MaterialExpressionNamedRerouteUsage_12}
   ]
   properties {constB=12}
  }
  MaterialExpressionMultiply_6 {
   class=MaterialExpressionMultiply
   inputs [
    A {from=MaterialExpressionComponentMask_23}
    B {from=MaterialExpressionOneMinus_0}
   ]
  }
  MaterialExpressionMultiply_7 {
   class=MaterialExpressionMultiply
   inputs [
    A {from=MaterialExpressionTime_1}
    B {from=""}
   ]
   properties {constB=0.1}
  }
  MaterialExpressionMultiply_8 {
   class=MaterialExpressionMultiply
   inputs [
    A {from=MaterialExpressionNamedRerouteUsage_54}
    B {from=""}
   ]
   properties {constB=2}
  }
  MaterialExpressionNamedRerouteDeclaration_0 {
   class=MaterialExpressionNamedRerouteDeclaration
   inputs [None {from=MaterialExpressionFunctionInput_0}]
   properties {name=In_UV, nodeColor=<rgba 1 0 0.9609 1>}
  }
  MaterialExpressionNamedRerouteDeclaration_1 {
   class=MaterialExpressionNamedRerouteDeclaration
   inputs [None {from=MaterialExpressionConvert_1}]
   properties {name=In_Bands, nodeColor=<rgba 0 1 0.7188 1>}
  }
  MaterialExpressionNamedRerouteDeclaration_10 {
   class=MaterialExpressionNamedRerouteDeclaration
   inputs [None {from=MaterialExpressionSubtract_1}]
   properties {name=d, nodeColor=<rgba 1 0 0.1641 1>}
  }
  MaterialExpressionNamedRerouteDeclaration_11 {
   class=MaterialExpressionNamedRerouteDeclaration
   inputs [None {from=MaterialExpressionMultiply_3}]
   properties {name=GridLines, nodeColor=<rgba 0 0.4844 1 1>}
  }
  MaterialExpressionNamedRerouteDeclaration_12 {
   class=MaterialExpressionNamedRerouteDeclaration
   inputs [None {from=MaterialExpressionLinearInterpolate_1}]
   properties {name=FinalColor, nodeColor=<rgba 0 1 0.4522 1>}
  }
  MaterialExpressionNamedRerouteDeclaration_15 {
   class=MaterialExpressionNamedRerouteDeclaration
   inputs [None {from=MaterialExpressionNamedRerouteDeclaration_0}]
   properties {name=ScaledUVs, nodeColor=<rgba 1 0.0703 0 1>}
  }
  MaterialExpressionNamedRerouteDeclaration_16 {
   class=MaterialExpressionNamedRerouteDeclaration
   inputs [None {from=MaterialExpressionSaturate_1}]
   properties {name=Levels, nodeColor=<rgba 0 0.4609 1 1>}
  }
  MaterialExpressionNamedRerouteDeclaration_2 {
   class=MaterialExpressionNamedRerouteDeclaration
   inputs [None {from=MaterialExpressionConvert_1}]
   properties {name=In_Segments, nodeColor=<rgba 1 0.4895 0 1>}
  }
  MaterialExpressionNamedRerouteDeclaration_23 {
   class=MaterialExpressionNamedRerouteDeclaration
   inputs [None {from=MaterialExpressionMaterialFunctionCall_6, output=0-1 mask}]
   properties {name=MaskOut, nodeColor=<rgba 1 0.1406 0 1>}
  }
  MaterialExpressionNamedRerouteDeclaration_24 {
   class=MaterialExpressionNamedRerouteDeclaration
   inputs [None {from=MaterialExpressionSaturate_5}]
   properties {name=BWMask, nodeColor=<rgba 0.7812 1 0 1>}
  }
  MaterialExpressionNamedRerouteDeclaration_25 {
   class=MaterialExpressionNamedRerouteDeclaration
   inputs [None {from=MaterialExpressionFunctionInput_8}]
   properties {name=In_Spectrum, nodeColor=<rgba 0.7344 0 1 1>}
  }
  MaterialExpressionNamedRerouteDeclaration_28 {
   class=MaterialExpressionNamedRerouteDeclaration
   inputs [None {from=MaterialExpressionLinearInterpolate_3}]
   properties {name=JustBars, nodeColor=<rgba 1 0 0.4453 1>}
  }
  MaterialExpressionNamedRerouteDeclaration_29 {
   class=MaterialExpressionNamedRerouteDeclaration
   inputs [None {from=MaterialExpressionLinearInterpolate_2}]
   properties {name=SoundBars_Reactivity, nodeColor=<rgba 0.4766 0 1 1>}
  }
  MaterialExpressionNamedRerouteDeclaration_3 {
   class=MaterialExpressionNamedRerouteDeclaration
   inputs [None {from=MaterialExpressionAppendVector_0}]
   properties {name=p, nodeColor=<rgba 0.0781 0 1 1>}
  }
  MaterialExpressionNamedRerouteDeclaration_31 {
   class=MaterialExpressionNamedRerouteDeclaration
   inputs [None {from=MaterialExpressionFunctionInput_9}]
   properties {name=In_Amplitude, nodeColor=<rgba 0.4526 1 0.0002 1>}
  }
  MaterialExpressionNamedRerouteDeclaration_32 {
   class=MaterialExpressionNamedRerouteDeclaration
   inputs [None {from=MaterialExpressionFunctionInput_10}]
   properties {name=In_TimeStamp, nodeColor=<rgba 1 0.2226 0.1568 1>}
  }
  MaterialExpressionNamedRerouteDeclaration_33 {
   class=MaterialExpressionNamedRerouteDeclaration
   inputs [None {from=MaterialExpressionFunctionInput_12}]
   properties {name=In_JustBars, nodeColor=<rgba 0 1 0.3906 1>}
  }
  MaterialExpressionNamedRerouteDeclaration_4 {
   class=MaterialExpressionNamedRerouteDeclaration
   inputs [None {from=MaterialExpressionTextureSampleParameter2D_0, output=RGB}]
   properties {name=Noise, nodeColor=<rgba 0.2422 1 0 1>}
  }
  MaterialExpressionNamedRerouteDeclaration_5 {
   class=MaterialExpressionNamedRerouteDeclaration
   inputs [None {from=MaterialExpressionFunctionInput_4}]
   properties {name=In_LowGraphColor, nodeColor=<rgba 1 0 0 1>}
  }
  MaterialExpressionNamedRerouteDeclaration_6 {
   class=MaterialExpressionNamedRerouteDeclaration
   inputs [None {from=MaterialExpressionFunctionInput_5}]
   properties {name=In_MiddleGraphColor, nodeColor=<rgba 0 1 0 1>}
  }
  MaterialExpressionNamedRerouteDeclaration_7 {
   class=MaterialExpressionNamedRerouteDeclaration
   inputs [None {from=MaterialExpressionFunctionInput_6}]
   properties {name=In_HighGraphColor, nodeColor=<rgba 0 0 1 1>}
  }
  MaterialExpressionNamedRerouteDeclaration_8 {
   class=MaterialExpressionNamedRerouteDeclaration
   inputs [None {from=MaterialExpressionMaterialFunctionCall_0, output=Result}]
   properties {name=LEDColor, nodeColor=<rgba 1 0.9137 0.002 1>}
  }
  MaterialExpressionNamedRerouteUsage_0 {class=MaterialExpressionNamedRerouteUsage}
  MaterialExpressionNamedRerouteUsage_1 {class=MaterialExpressionNamedRerouteUsage}
  MaterialExpressionNamedRerouteUsage_11 {class=MaterialExpressionNamedRerouteUsage}
  MaterialExpressionNamedRerouteUsage_12 {class=MaterialExpressionNamedRerouteUsage}
  MaterialExpressionNamedRerouteUsage_13 {class=MaterialExpressionNamedRerouteUsage}
  MaterialExpressionNamedRerouteUsage_14 {class=MaterialExpressionNamedRerouteUsage}
  MaterialExpressionNamedRerouteUsage_15 {class=MaterialExpressionNamedRerouteUsage}
  MaterialExpressionNamedRerouteUsage_16 {class=MaterialExpressionNamedRerouteUsage}
  MaterialExpressionNamedRerouteUsage_17 {class=MaterialExpressionNamedRerouteUsage}
  MaterialExpressionNamedRerouteUsage_18 {class=MaterialExpressionNamedRerouteUsage}
  MaterialExpressionNamedRerouteUsage_19 {class=MaterialExpressionNamedRerouteUsage}
  MaterialExpressionNamedRerouteUsage_2 {class=MaterialExpressionNamedRerouteUsage}
  MaterialExpressionNamedRerouteUsage_25 {class=MaterialExpressionNamedRerouteUsage}
  MaterialExpressionNamedRerouteUsage_26 {class=MaterialExpressionNamedRerouteUsage}
  MaterialExpressionNamedRerouteUsage_27 {class=MaterialExpressionNamedRerouteUsage}
  MaterialExpressionNamedRerouteUsage_29 {class=MaterialExpressionNamedRerouteUsage}
  MaterialExpressionNamedRerouteUsage_3 {class=MaterialExpressionNamedRerouteUsage}
  MaterialExpressionNamedRerouteUsage_34 {class=MaterialExpressionNamedRerouteUsage}
  MaterialExpressionNamedRerouteUsage_37 {class=MaterialExpressionNamedRerouteUsage}
  MaterialExpressionNamedRerouteUsage_4 {class=MaterialExpressionNamedRerouteUsage}
  MaterialExpressionNamedRerouteUsage_40 {class=MaterialExpressionNamedRerouteUsage}
  MaterialExpressionNamedRerouteUsage_41 {class=MaterialExpressionNamedRerouteUsage}
  MaterialExpressionNamedRerouteUsage_44 {class=MaterialExpressionNamedRerouteUsage}
  MaterialExpressionNamedRerouteUsage_45 {class=MaterialExpressionNamedRerouteUsage}
  MaterialExpressionNamedRerouteUsage_46 {class=MaterialExpressionNamedRerouteUsage}
  MaterialExpressionNamedRerouteUsage_47 {class=MaterialExpressionNamedRerouteUsage}
  MaterialExpressionNamedRerouteUsage_49 {class=MaterialExpressionNamedRerouteUsage}
  MaterialExpressionNamedRerouteUsage_5 {class=MaterialExpressionNamedRerouteUsage}
  MaterialExpressionNamedRerouteUsage_50 {class=MaterialExpressionNamedRerouteUsage}
  MaterialExpressionNamedRerouteUsage_51 {class=MaterialExpressionNamedRerouteUsage}
  MaterialExpressionNamedRerouteUsage_52 {class=MaterialExpressionNamedRerouteUsage}
  MaterialExpressionNamedRerouteUsage_53 {class=MaterialExpressionNamedRerouteUsage}
  MaterialExpressionNamedRerouteUsage_54 {class=MaterialExpressionNamedRerouteUsage}
  MaterialExpressionNamedRerouteUsage_56 {class=MaterialExpressionNamedRerouteUsage}
  MaterialExpressionNamedRerouteUsage_57 {class=MaterialExpressionNamedRerouteUsage}
  MaterialExpressionNamedRerouteUsage_59 {class=MaterialExpressionNamedRerouteUsage}
  MaterialExpressionNamedRerouteUsage_6 {class=MaterialExpressionNamedRerouteUsage}
  MaterialExpressionNamedRerouteUsage_60 {class=MaterialExpressionNamedRerouteUsage}
  MaterialExpressionNamedRerouteUsage_7 {class=MaterialExpressionNamedRerouteUsage}
  MaterialExpressionNamedRerouteUsage_8 {class=MaterialExpressionNamedRerouteUsage}
  MaterialExpressionNamedRerouteUsage_9 {class=MaterialExpressionNamedRerouteUsage}
  MaterialExpressionOneMinus_0 {class=MaterialExpressionOneMinus, inputs [None {from=MaterialExpressionSaturate_5}]}
  MaterialExpressionPanner_0 {
   class=MaterialExpressionPanner
   inputs [
    Coordinate {from=MaterialExpressionRotator_0}
    Time {from=""}
    Speed {from=""}
   ]
  }
  MaterialExpressionPower_0 {
   class=MaterialExpressionPower
   inputs [
    Base {from=MaterialExpressionNamedRerouteUsage_11}
    Exp {from=MaterialExpressionConstant_14}
   ]
  }
  MaterialExpressionRotator_0 {
   class=MaterialExpressionRotator
   inputs [
    Coordinate {from=MaterialExpressionTextureCoordinate_0, output=""}
    Time {from=MaterialExpressionMaterialFunctionCall_2, output=Result}
   ]
   properties {}
  }
  MaterialExpressionSaturate_1 {class=MaterialExpressionSaturate, inputs [None {from=MaterialExpressionLinearInterpolate_0}]}
  MaterialExpressionSaturate_5 {class=MaterialExpressionSaturate, inputs [None {from=MaterialExpressionIf_1}]}
  MaterialExpressionSaturate_6 {class=MaterialExpressionSaturate, inputs [None {from=MaterialExpressionDivide_9}]}
  MaterialExpressionSmoothStep_0 {
   class=MaterialExpressionSmoothStep
   inputs [
    Min {from=MaterialExpressionConstant_5, output=""}
    Max {from=MaterialExpressionConstant_6, output=""}
    Value {from=MaterialExpressionMaterialFunctionCall_1, output=R}
   ]
   properties {constMax=0.35, constMin=0.5}
  }
  MaterialExpressionSmoothStep_1 {
   class=MaterialExpressionSmoothStep
   inputs [
    Min {from=MaterialExpressionConstant_5, output=""}
    Max {from=MaterialExpressionConstant_6, output=""}
    Value {from=MaterialExpressionMaterialFunctionCall_1, output=G}
   ]
   properties {}
  }
  MaterialExpressionSquareRoot_0 {class=MaterialExpressionSquareRoot, inputs [None {from=MaterialExpressionComponentMask_1, output=""}], properties {}}
  MaterialExpressionSubtract_0 {
   class=MaterialExpressionSubtract
   inputs [
    A {from=MaterialExpressionNamedRerouteUsage_27}
    B {from=MaterialExpressionNamedRerouteUsage_13}
   ]
  }
  MaterialExpressionSubtract_1 {
   class=MaterialExpressionSubtract
   inputs [
    A {from=MaterialExpressionFrac_0}
    B {from=""}
   ]
   properties {constB=0.5}
  }
  MaterialExpressionSwitch_1 {
   class=MaterialExpressionSwitch
   inputs [
    SwitchValue {from=MaterialExpressionMultiply_5, output=""}
    Default {from="", output=""}
    None {from=MaterialExpressionMaterialFunctionCall_7, output=R}
    None {from=MaterialExpressionMaterialFunctionCall_7, output=R}
    None {from=MaterialExpressionMaterialFunctionCall_7, output=R}
    None {from=MaterialExpressionMaterialFunctionCall_7, output=R}
   ]
   properties {inputs [
     {
      inputName=None
      input {
       expression=&self:MaterialExpressionMaterialFunctionCall_7
       inputName=None
       mask=0
       maskA=0
       maskB=0
       maskG=0
       maskR=0
       outputIndex=0
      }
     }
     {
      inputName=None
      input {
       expression=&self:MaterialExpressionMaterialFunctionCall_7
       inputName=None
       mask=0
       maskA=0
       maskB=0
       maskG=0
       maskR=0
       outputIndex=1
      }
     }
     {
      inputName=None
      input {
       expression=&self:MaterialExpressionMaterialFunctionCall_7
       inputName=None
       mask=0
       maskA=0
       maskB=0
       maskG=0
       maskR=0
       outputIndex=2
      }
     }
     {
      inputName=None
      input {
       expression=&self:MaterialExpressionMaterialFunctionCall_7
       inputName=None
       mask=0
       maskA=0
       maskB=0
       maskG=0
       maskR=0
       outputIndex=3
      }
     }
    ]}
  }
  MaterialExpressionTextureCoordinate_0 {class=MaterialExpressionTextureCoordinate}
  MaterialExpressionTextureSampleParameter2D_0 {
   class=MaterialExpressionTextureSampleParameter2D
   inputs [
    UVs {from=MaterialExpressionAppendVector_1, output=""}
    Bias {from="", output=""}
    Apply View MipBias {from="", output=""}
   ]
   properties {
    constMipValue=-6
    group=LED Spectrum Bar Graph
    mipValueMode=TMVM_MipBias
    parameterName=NoiseTexture
    samplerSource=SSM_Wrap_WorldGroupSettings
    texture=/BRCosmetics/Animation/Props/Gadgets_RadioPaca/Textures/T_NoiseForBarGraphs
   }
  }
  MaterialExpressionTime_1 {class=MaterialExpressionTime}
 }}
