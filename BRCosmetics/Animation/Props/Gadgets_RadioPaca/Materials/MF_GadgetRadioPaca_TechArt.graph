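# Material function that adds audio-reactive LED spectrum displays and glow-stick emissive to the RadioPaca gadget prop's base color and emissive.
# Scaled, rotated or width-skewed UV regions for the DJ deck, visor and Paca sides, front and back are lerped with per-region band counts.
# These feed a single MF_2DLEDSpectrum call alongside audio amplitude, spectrum, colors and a transition timestamp.
# Two tech-art mask inputs gate effects, including a BPM-timed pulse.
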MaterialFunction /BRCosmetics/Animation/Props/Gadgets_RadioPaca/Materials/MF_GadgetRadioPaca_TechArt
graph {nodes [
  n1 = GetMaterialAttributes(None=NamedRerouteUsage()) {attributeGetTypes [
    #1
    #2
    #3
    #4
    #5
   ]}
  n2 = SetMaterialAttributes!(
   MaterialAttributes=@n1.MaterialAttributes
   EmissiveColor=NamedRerouteUsage()
   BaseColor=NamedRerouteUsage()
   Specular=@n1.MaterialAttributes
   Roughness=@n1.MaterialAttributes
  ) {attributeSetTypes [
    #1
    #2
    #3
    #4
   ]}
  FunctionOutput(None=@n2)
  NamedRerouteDeclaration(None=FunctionInput(Preview=TextureSample() {samplerType=SAMPLERTYPE_Masks, texture=/BRCosmetics/Animation/Props/Gadgets_RadioPaca/Textures/T_RadioPaca_Gadgets_FX}.RGB) {bUsePreviewValueAsDefault=true, inputName=In_TechArtMask}) {name=In_TechArtMask, nodeColor=<rgba 0.4531 0 1 1>}
  NamedRerouteDeclaration(None=FunctionInput() {inputName=In_MaterialAttributes, inputType=FunctionInput_MaterialAttributes}) {name=In_MaterialAttributes, nodeColor=<rgba 0 1 0.1328 1>}
  n3 = ComponentMask(None=NamedRerouteUsage()) {b=true}
  n4 = Multiply(A=Time(), B=Divide(A=ScalarParameter() {defaultValue=120, parameterName=Deafault BPM}) {constB=60})
  n5 = ConstantBiasScale(None=Frac(None=NamedRerouteDeclaration(None=@n4) {name=DefaultBPM, nodeColor=<rgba 0.0547 1 0 1>})) {bias=0, scale=2}
  n6 = NamedRerouteDeclaration(None=OneMinus(None=Abs(None=@n5))) {name=DefaultBeat, nodeColor=<rgba 0 1 0.5078 1>}
  n7 = Saturate(None=Multiply(A=@n6, B=@n6))
  NamedRerouteDeclaration(None=Multiply(A=Multiply(A=@n3, B=@n7))) {name=StuffTimedToBeat, nodeColor=<rgba 0.5469 0 1 1>}
  NamedRerouteDeclaration(None=AppendVector(
    A=ScalarParameter() {defaultValue=0.404, group=RadioPaca Tech Art, parameterName=ScaleX_FrontofDJDeck}
    B=ScalarParameter() {defaultValue=0.086, group=RadioPaca Tech Art, parameterName=ScaleY_FrontofDJDeck}
   )) {name=ScaleUVFrontofDJDeck, nodeColor=<rgba 0.5469 0 1 1>}
  NamedRerouteDeclaration(None=LinearInterpolate(
    A=VectorParameter() {group=RadioPaca Tech Art, parameterName=LowColor_Paca, defaultValue=<rgba 0.0561 0.1248 0.1878 1>}.RGB
    B=VectorParameter() {group=RadioPaca Tech Art, parameterName=LowColor_FrontofDJDeck, defaultValue=<rgba 0 1 1 1>}.RGB
    Alpha=NamedRerouteUsage()
   )) {name=LowColor_Paca, nodeColor=<rgba 1 0.6328 0 1>}
  NamedRerouteDeclaration(None=LinearInterpolate(
    A=VectorParameter() {group=RadioPaca Tech Art, parameterName=MidColor_Paca, defaultValue=<rgba 0.0123 0.0865 0.4179 1>}.RGB
    B=VectorParameter() {group=RadioPaca Tech Art, parameterName=MidColor_FrontofDJDeck, defaultValue=<rgba 0 1 1 1>}.RGB
    Alpha=NamedRerouteUsage()
   )) {name=MidColor_Paca, nodeColor=<rgba 0.3125 0 1 1>}
  n8 = Divide(A=Time(), B=ScalarParameter() {defaultValue=3, group=FNCSShield_Screen, parameterName=DebugTransitionLength, sortPriority=30})
  n9 = MaterialFunctionCall(
   TimestampA=ScalarParameter() {defaultValue=1, parameterName=Timestamp, sortPriority=30}
   TransitionLength_In=ScalarParameter() {defaultValue=0.1, parameterName=TransitionLength, sortPriority=30}
  ) {materialFunction=&MF_Timestamp_InOut_01}
  NamedRerouteDeclaration(None=StaticSwitchParameter(True=Frac(None=@n8), False=VertexInterpolator(VS=@n9.Linear).PS) {group=FNCSShield_Screen_Graphics, parameterName=DebugModeOn, sortPriority=0}) {name=TimeStamp, nodeColor=<rgba 0.125 1 0 1>}
  NamedRerouteDeclaration(None=@n1.EmissiveColor) {name=In_Emissive, nodeColor=<rgba 1 0.5156 0 1>}
  NamedRerouteDeclaration(None=FunctionInput(Preview=TextureSample() {texture=/BRCosmetics/Animation/Props/Gadgets_RadioPaca/Textures/T_RadioPaca_Gadgets_FX2}.RGB) {bUsePreviewValueAsDefault=true, inputName=In_TechArtMask2}) {name=In_TechArtMask2, nodeColor=<rgba 0.4531 0 1 1>}
  n10 = NamedRerouteUsage()
  NamedRerouteDeclaration(None=Multiply(
    A=VectorParameter() {group=RadioPaca Tech Art, parameterName=GlowStick, defaultValue=<rgba 0.0883 3 0.08 1>}.RGB
    B=ComponentMask(None=@n10) {b=true}
   )) {name=GlowStickEmissiveColor, nodeColor=<rgba 0.1534 1 0.078 1>}
  n11 = Subtract(A=TextureCoordinate(), B=NamedRerouteUsage())
  n12 = Constant2Vector() {g=0.5, r=0.5}
  n13 = MaterialFunctionCall(Float2=Subtract(A=@n11, B=@n12)) {materialFunction=&BreakOutFloat2Components}
  n14 = NamedRerouteUsage()
  n15 = ScalarParameter() {group=RadioPaca Tech Art, parameterName=SideAMidValue}
  n16 = NamedRerouteUsage()
  n17 = NamedRerouteUsage()
  n18 = Multiply(
   A=@n13.R
   B=MaterialFunctionCall!(
    1=ScalarParameter() {defaultValue=0.632, group=RadioPaca Tech Art, parameterName=SideAWidthA}
    2=ScalarParameter() {defaultValue=0.39, group=RadioPaca Tech Art, parameterName=SideAWidthB}
    3=ScalarParameter() {defaultValue=0.39, group=RadioPaca Tech Art, parameterName=SideAWidthC}
    4
    A=If(A=@n14, B=@n15, "A > B"=@n16, "A == B"=@n17, "A < B"=@n17)
   ) {materialFunction=/Engine/Functions/Engine_MaterialFunctions02/Lerp_Multiple_Float}.Lerp 3 Inputs
  )
  n19 = ScalarParameter() {defaultValue=3.5211, group=RadioPaca Tech Art, parameterName=ScaleX_SideofPacaA}
  n20 = ScalarParameter() {defaultValue=2, group=RadioPaca Tech Art, parameterName=ScaleY_SideofPacaA}
  n21 = Add(A=Multiply(A=AppendVector(A=@n18, B=@n13.G), B=AppendVector(A=@n19, B=@n20)) {constB=2}, B=@n12) {constA=0.5, constB=0.5}
  n22 = MaterialFunctionCall(Multiplier=Constant() {r=4}) {materialFunction=&Pi}
  NamedRerouteDeclaration(None=Add(A=Rotator(Coordinate=@n21, Time=@n22.Result), B=NamedRerouteUsage())) {name=SkewedUVsSideA, nodeColor=<rgba 0.8177 0.5911 0.0894 1>}
  NamedRerouteDeclaration(None=AppendVector(
    A=ScalarParameter() {defaultValue=-0.16, group=RadioPaca Tech Art, parameterName=OffsetX_SideofPacaA}
    B=ScalarParameter() {group=RadioPaca Tech Art, parameterName=OffsetY_SideofPacaA}
   )) {name=OffsetSideA, nodeColor=<rgba 1 0.0661 0.0035 1>}
  n23 = Subtract(A=TextureCoordinate(), B=NamedRerouteUsage())
  n24 = Constant2Vector() {g=0.5, r=0.5}
  n25 = MaterialFunctionCall(Float2=Subtract(A=@n23, B=@n24)) {materialFunction=&BreakOutFloat2Components}
  n26 = ScalarParameter() {defaultValue=0.632, group=RadioPaca Tech Art, parameterName=SideBWidthA}
  n27 = ScalarParameter() {defaultValue=0.39, group=RadioPaca Tech Art, parameterName=SideBWidthB}
  n28 = ScalarParameter() {defaultValue=0.39, group=RadioPaca Tech Art, parameterName=SideBWidthC}
  n29 = NamedRerouteUsage()
  n30 = If(
   A=NamedRerouteUsage()
   B=ScalarParameter() {group=RadioPaca Tech Art, parameterName=SideBMidValue}
   "A > B"=NamedRerouteUsage()
   "A == B"=@n29
   "A < B"=@n29
  )
  n31 = AppendVector(
   A=Multiply(
    A=@n25.R
    B=MaterialFunctionCall(1=@n26, 2=@n27, 3=@n28, A=@n30) {materialFunction=&Lerp_Multiple_Float}.Lerp 3 Inputs
   )
   B=@n25.G
  )
  n32 = AppendVector(
   A=ScalarParameter() {defaultValue=1, group=RadioPaca Tech Art, parameterName=ScaleX_SideofPacaB}
   B=ScalarParameter() {defaultValue=1, group=RadioPaca Tech Art, parameterName=ScaleY_SideofPacaB}
  )
  n33 = Rotator(
   Coordinate=Add(A=Multiply(A=@n31, B=@n32) {constB=2}, B=@n24) {constA=0.5, constB=0.5}
   Time=MaterialFunctionCall(Multiplier=Constant() {r=4}) {materialFunction=&Pi}.Result
  )
  n34 = NamedRerouteUsage()
  n35 = VectorParameter() {group=RadioPaca Tech Art, parameterName=SideBExtraOffset, defaultValue=<rgba 0 0 0 1>}
  NamedRerouteDeclaration(None=Add(A=Add(A=@n33, B=@n34), B=ComponentMask(None=@n35.RGB) {g=true, r=true})) {name=SkewedUVsSideB, nodeColor=<rgba 1 0 0.0938 1>}
  NamedRerouteDeclaration(None=AppendVector(
    A=ScalarParameter() {defaultValue=0.587, group=RadioPaca Tech Art, parameterName=OffsetX_SideofPacaB}
    B=ScalarParameter() {defaultValue=0.563, group=RadioPaca Tech Art, parameterName=OffsetY_SideofPacaB}
   )) {name=OffsetSideB, nodeColor=<rgba 0 1 0.9531 1>}
  NamedRerouteDeclaration(None=Add(A=@n25.G, B=ComponentMask(None=@n23) {g=true})) {name=OffsetRangeB, nodeColor=<rgba 1 0 0.4219 1>}
  NamedRerouteDeclaration(None=MaterialFunctionCall(
    Input=NamedRerouteUsage()
    Input Low=Constant() {r=-0.5}
    Input High=Constant()
    Target Low=Constant() {r=-0.5}
    Target High=ScalarParameter() {group=RadioPaca Tech Art, parameterName=SideBMidValue}
   ) {materialFunction=&RemapValueRange}.Result) {name=LowerRangeSideB, nodeColor=<rgba 0 0.7422 1 1>}
  NamedRerouteDeclaration(None=MaterialFunctionCall(
    Input=NamedRerouteUsage()
    Input Low=Constant()
    Input High=Constant() {r=0.5}
    Target Low=ScalarParameter() {group=RadioPaca Tech Art, parameterName=SideBMidValue}
    Target High=Constant() {r=0.5}
   ) {materialFunction=&RemapValueRange}.Result) {name=UpperRangeSideB, nodeColor=<rgba 1 0.9375 0 1>}
  NamedRerouteDeclaration(None=@n1.BaseColor) {name=In_BaseColor, nodeColor=<rgba 0.1953 0 1 1>}
  NamedRerouteDeclaration(None=MaterialFunctionCall(
    Input=NamedRerouteUsage()
    Input Low=Constant() {r=-0.5}
    Input High=Constant()
    Target Low=Constant() {r=-0.5}
    Target High=ScalarParameter() {group=RadioPaca Tech Art, parameterName=SideAMidValue}
   ) {materialFunction=&RemapValueRange}.Result) {name=LowerRangeSideA, nodeColor=<rgba 0 0.7422 1 1>}
  NamedRerouteDeclaration(None=MaterialFunctionCall(
    Input=NamedRerouteUsage()
    Input Low=Constant()
    Input High=Constant() {r=0.5}
    Target Low=ScalarParameter() {group=RadioPaca Tech Art, parameterName=SideAMidValue}
    Target High=Constant() {r=0.5}
   ) {materialFunction=&RemapValueRange}.Result) {name=UpperRangeSideA, nodeColor=<rgba 1 0.9375 0 1>}
  NamedRerouteDeclaration(None=Add(A=@n13.G, B=ComponentMask(None=@n11) {g=true})) {name=OffsetRangeA, nodeColor=<rgba 1 0.2578 0 1>}
  n36 = Subtract(A=TextureCoordinate(), B=NamedRerouteUsage())
  n37 = Constant2Vector() {g=0.5, r=0.5}
  n38 = MaterialFunctionCall(Float2=Subtract(A=@n36, B=@n37)) {materialFunction=&BreakOutFloat2Components}
  n39 = ScalarParameter() {defaultValue=0.632, group=RadioPaca Tech Art, parameterName=FrontWidthA}
  n40 = ScalarParameter() {defaultValue=0.39, group=RadioPaca Tech Art, parameterName=FrontWidthB}
  n41 = ScalarParameter() {defaultValue=0.39, group=RadioPaca Tech Art, parameterName=FrontWidthC}
  n42 = NamedRerouteUsage()
  n43 = If(
   A=NamedRerouteUsage()
   B=ScalarParameter() {group=RadioPaca Tech Art, parameterName=FrontMidValue}
   "A > B"=NamedRerouteUsage()
   "A == B"=@n42
   "A < B"=@n42
  )
  n44 = AppendVector(
   A=Multiply(
    A=@n38.R
    B=MaterialFunctionCall(1=@n39, 2=@n40, 3=@n41, A=@n43) {materialFunction=&Lerp_Multiple_Float}.Lerp 3 Inputs
   )
   B=@n38.G
  )
  n45 = AppendVector(
   A=ScalarParameter() {defaultValue=1, group=RadioPaca Tech Art, parameterName=ScaleX_FrontofPacaB}
   B=ScalarParameter() {defaultValue=1, group=RadioPaca Tech Art, parameterName=ScaleY_FrontofPacaB}
  )
  n46 = Rotator(
   Coordinate=Add(A=Multiply(A=@n44, B=@n45) {constB=2}, B=@n37) {constA=0.5, constB=0.5}
   Time=MaterialFunctionCall(Multiplier=Constant() {r=4}) {materialFunction=&Pi}.Result
  )
  n47 = NamedRerouteUsage()
  n48 = VectorParameter() {group=RadioPaca Tech Art, parameterName=FrontExtraOffset, defaultValue=<rgba 0 0 0 1>}
  NamedRerouteDeclaration(None=Add(A=Add(A=@n46, B=@n47), B=ComponentMask(None=@n48.RGB) {g=true, r=true})) {name=SkewedUVsFront, nodeColor=<rgba 0.0307 0.0284 0.0723 1>}
  NamedRerouteDeclaration(None=AppendVector(
    A=ScalarParameter() {defaultValue=0.587, group=RadioPaca Tech Art, parameterName=OffsetX_FrontofPacaB}
    B=ScalarParameter() {defaultValue=0.563, group=RadioPaca Tech Art, parameterName=OffsetY_FrontofPacaB}
   )) {name=OffsetSideB_1, nodeColor=<rgba 0 1 0.9531 1>}
  NamedRerouteDeclaration(None=Add(A=@n38.G, B=ComponentMask(None=@n36) {g=true})) {name=OffsetRangeB_1, nodeColor=<rgba 1 0 0.4219 1>}
  NamedRerouteDeclaration(None=MaterialFunctionCall(
    Input=NamedRerouteUsage()
    Input Low=Constant() {r=-0.5}
    Input High=Constant()
    Target Low=Constant() {r=-0.5}
    Target High=ScalarParameter() {group=RadioPaca Tech Art, parameterName=FrontMidValue}
   ) {materialFunction=&RemapValueRange}.Result) {name=LowerRangeSideB_1, nodeColor=<rgba 0 0.7422 1 1>}
  NamedRerouteDeclaration(None=MaterialFunctionCall(
    Input=NamedRerouteUsage()
    Input Low=Constant()
    Input High=Constant() {r=0.5}
    Target Low=ScalarParameter() {group=RadioPaca Tech Art, parameterName=FrontMidValue}
    Target High=Constant() {r=0.5}
   ) {materialFunction=&RemapValueRange}.Result) {name=UpperRangeSideB_1, nodeColor=<rgba 1 0.9375 0 1>}
  n49 = Subtract(A=TextureCoordinate(), B=NamedRerouteUsage())
  n50 = Constant2Vector() {g=0.5, r=0.5}
  n51 = MaterialFunctionCall(Float2=Subtract(A=@n49, B=@n50)) {materialFunction=&BreakOutFloat2Components}
  n52 = ScalarParameter() {defaultValue=0.632, group=RadioPaca Tech Art, parameterName=BackWidthA}
  n53 = ScalarParameter() {defaultValue=0.39, group=RadioPaca Tech Art, parameterName=BackWidthB}
  n54 = ScalarParameter() {defaultValue=0.39, group=RadioPaca Tech Art, parameterName=BackWidthC}
  n55 = NamedRerouteUsage()
  n56 = If(
   A=NamedRerouteUsage()
   B=ScalarParameter() {group=RadioPaca Tech Art, parameterName=BackMidValue}
   "A > B"=NamedRerouteUsage()
   "A == B"=@n55
   "A < B"=@n55
  )
  n57 = AppendVector(
   A=Multiply(
    A=@n51.R
    B=MaterialFunctionCall(1=@n52, 2=@n53, 3=@n54, A=@n56) {materialFunction=&Lerp_Multiple_Float}.Lerp 3 Inputs
   )
   B=@n51.G
  )
  n58 = AppendVector(
   A=ScalarParameter() {group=RadioPaca Tech Art, parameterName=ScaleX_BackofPaca}
   B=ScalarParameter() {defaultValue=0.093, group=RadioPaca Tech Art, parameterName=ScaleY_BackofPaca}
  )
  n59 = Rotator(
   Coordinate=Add(A=Multiply(A=@n57, B=@n58) {constB=2}, B=@n50) {constA=0.5, constB=0.5}
   Time=MaterialFunctionCall(Multiplier=Constant() {r=4}) {materialFunction=&Pi}.Result
  )
  n60 = NamedRerouteUsage()
  n61 = VectorParameter() {group=RadioPaca Tech Art, parameterName=BackExtraOffset, defaultValue=<rgba 0 0 0 1>}
  NamedRerouteDeclaration(None=Add(A=Add(A=@n59, B=@n60), B=ComponentMask(None=@n61.RGB) {g=true, r=true})) {name=SkewedUVsBack, nodeColor=<rgba 0.159 0.8385 0.7847 1>}
  NamedRerouteDeclaration(None=AppendVector(
    A=ScalarParameter() {defaultValue=1, group=RadioPaca Tech Art, parameterName=OffsetX_BackofPaca}
    B=ScalarParameter() {defaultValue=1, group=RadioPaca Tech Art, parameterName=OffsetY_BackofPaca}
   )) {name=OffsetSideB_2, nodeColor=<rgba 0 1 0.9531 1>}
  n62 = VectorParameter() {group=RadioPaca Tech Art, parameterName=BaseEmissive, defaultValue=<rgba 0 0 0 1>}
  n63 = NamedRerouteUsage()
  n64 = LinearInterpolate(A=Constant(), B=Multiply(A=Add(A=@n62.RGB, B=@n63)) {constB=2}, Alpha=ComponentMask(None=NamedRerouteUsage()) {g=true})
  n65 = NamedRerouteUsage()
  n66 = Constant()
  n67 = NamedRerouteUsage()
  n68 = NamedRerouteUsage()
  NamedRerouteDeclaration(None=Add(A=Add(A=@n64, B=@n65), B=LinearInterpolate(A=@n66, B=@n67, Alpha=@n68))) {name=Out_Emissive, nodeColor=<rgba 0.125 1 0 1>}
  NamedRerouteDeclaration(None=Add(A=@n51.G, B=ComponentMask(None=@n49) {g=true})) {name=OffsetRangeB_2, nodeColor=<rgba 1 0 0.4219 1>}
  NamedRerouteDeclaration(None=MaterialFunctionCall(
    Input=NamedRerouteUsage()
    Input Low=Constant() {r=-0.5}
    Input High=Constant()
    Target Low=Constant() {r=-0.5}
    Target High=ScalarParameter() {group=RadioPaca Tech Art, parameterName=BackMidValue}
   ) {materialFunction=&RemapValueRange}.Result) {name=LowerRangeSideB_2, nodeColor=<rgba 0 0.7422 1 1>}
  NamedRerouteDeclaration(None=MaterialFunctionCall(
    Input=NamedRerouteUsage()
    Input Low=Constant()
    Input High=Constant() {r=0.5}
    Target Low=ScalarParameter() {group=RadioPaca Tech Art, parameterName=BackMidValue}
    Target High=Constant() {r=0.5}
   ) {materialFunction=&RemapValueRange}.Result) {name=UpperRangeSideB_2, nodeColor=<rgba 1 0.9375 0 1>}
  n69 = VectorParameter() {group=RadioPaca Tech Art, parameterName=HighColor_FrontofDJDeck, defaultValue=<rgba 0.8469 0.0762 0.8469 1>}
  NamedRerouteDeclaration(None=LinearInterpolate(
    A=VectorParameter() {group=RadioPaca Tech Art, parameterName=HighColor_Paca, defaultValue=<rgba 0.0545 0.2335 0.9304 1>}.RGB
    B=@n69.RGB
    Alpha=NamedRerouteUsage()
   )) {name=HighColor_Paca, nodeColor=<rgba 0.0078 1 0 1>}
  NamedRerouteDeclaration(None=AppendVector(
    A=ScalarParameter() {defaultValue=0.52, group=RadioPaca Tech Art, parameterName=OffsetX_FrontofDJDeck}
    B=ScalarParameter() {defaultValue=0.904, group=RadioPaca Tech Art, parameterName=OffsetY_FrontofDJDeck}
   )) {name=OffsetUVFrontofDJDeck, nodeColor=<rgba 0 1 0.2266 1>}
  n70 = MaterialFunctionCall() {materialFunction=/Game/Packages/Fortress_SharedMaterials/Base_Material_Functions/BaseCharacterShaderFunctions/Audio/MF_AudioReactivity}
  NamedRerouteDeclaration(None=Multiply(A=ScalarParameter() {defaultValue=1, group=RadioPaca Tech Art, parameterName=AudioReactivityAmplitude}, B=@n70.Amplitude)) {name=Amplitude, nodeColor=<rgba 0 1 0.1094 1>}
  NamedRerouteDeclaration(None=Multiply(A=@n70.Spectrum, B=ScalarParameter() {defaultValue=1, group=RadioPaca Tech Art, parameterName=AudioReactivitySpectrum})) {name=Spectrum, nodeColor=<rgba 1 0 0.75 1>}
  n71 = NamedRerouteUsage()
  n72 = LinearInterpolate(A=NamedRerouteUsage(), Alpha=NamedRerouteUsage()) {constB=0}
  n73 = ComponentMask(None=NamedRerouteUsage()) {g=true}
  n74 = NamedRerouteUsage()
  n75 = ComponentMask(None=NamedRerouteUsage()) {g=true}
  NamedRerouteDeclaration(None=LinearInterpolate(A=LinearInterpolate(A=@n71, B=@n72, Alpha=@n73), B=NamedRerouteUsage(), Alpha=Multiply(A=@n74, B=@n75))) {name=Out_BaseColor, nodeColor=<rgba 1 0 0.4453 1>}
  NamedRerouteDeclaration(None=Power(Base=NamedRerouteUsage(), Exp=ScalarParameter() {defaultValue=8, group=RadioPaca Tech Art, parameterName=AmplitudeEmissiveExp}) {constExponent=8}) {name=AmplitudeAdjusted, nodeColor=<rgba 0 1 0.9531 1>}
  n76 = TextureCoordinate()
  n77 = MaterialFunctionCall(Multiplier=Constant() {r=6}) {materialFunction=&Pi}
  n78 = Divide(A=Subtract(A=Rotator(Coordinate=@n76, Time=@n77.Result), B=NamedRerouteUsage()), B=NamedRerouteUsage())
  NamedRerouteDeclaration(None=MaterialFunctionCall(UVs=NamedRerouteDeclaration(None=@n78) {name=ScaledUVs_Deck, nodeColor=<rgba 1 0 0.0937 1>}) {materialFunction=&ScaleUVsByCenter}.0-1 mask) {name=Mask_FrontofDJDeck, nodeColor=<rgba 0 0.4141 1 1>}
  NamedRerouteDeclaration(None=AppendVector(
    A=ScalarParameter() {defaultValue=0.08, group=RadioPaca Tech Art, parameterName=ScaleX_TopofDJDeck}
    B=ScalarParameter() {defaultValue=0.093, group=RadioPaca Tech Art, parameterName=ScaleY_TopofDJDeck}
   )) {name=ScaleUVDeckTop, nodeColor=<rgba 0.9453 0 1 1>}
  NamedRerouteDeclaration(None=AppendVector(
    A=ScalarParameter() {defaultValue=0.587, group=RadioPaca Tech Art, parameterName=OffsetX_TopofDJDeck}
    B=ScalarParameter() {defaultValue=0.563, group=RadioPaca Tech Art, parameterName=OffsetY_TopofDJDeck}
   )) {name=OffsetUVDeckTop, nodeColor=<rgba 0 1 0.625 1>}
  n79 = TextureCoordinate()
  n80 = Constant()
  n81 = Divide(A=Subtract(A=Rotator(Coordinate=@n79, Time=@n80), B=NamedRerouteUsage()), B=NamedRerouteUsage())
  NamedRerouteDeclaration(None=MaterialFunctionCall(UVs=NamedRerouteDeclaration(None=@n81) {name=ScaledUV_DeckTop, nodeColor=<rgba 0.7344 1 0 1>}) {materialFunction=&ScaleUVsByCenter}.0-1 mask) {name=Mask_DeckTop, nodeColor=<rgba 1 0.3047 0 1>}
  NamedRerouteDeclaration(None=MaterialFunctionCall(UVs=NamedRerouteUsage()) {materialFunction=&ScaleUVsByCenter}.0-1 mask) {name=Mask_SideOfDeckA, nodeColor=<rgba 0 0.4141 1 1>}
  NamedRerouteDeclaration(None=MaterialFunctionCall(UVs=NamedRerouteUsage()) {materialFunction=&ScaleUVsByCenter}.0-1 mask) {name=Mask_SideOfDeckB, nodeColor=<rgba 0 0.4141 1 1>}
  NamedRerouteDeclaration(None=MaterialFunctionCall(UVs=NamedRerouteUsage()) {materialFunction=&ScaleUVsByCenter}.0-1 mask) {name=Mask_FrontOfPaca, nodeColor=<rgba 0 0.4141 1 1>}
  NamedRerouteDeclaration(None=MaterialFunctionCall(UVs=NamedRerouteUsage()) {materialFunction=&ScaleUVsByCenter}.0-1 mask) {name=Mask_BackofPaca, nodeColor=<rgba 0 0.4141 1 1>}
  NamedRerouteDeclaration(None=AppendVector(
    A=ScalarParameter() {group=RadioPaca Tech Art, parameterName=ScaleX_VisorPaca}
    B=ScalarParameter() {defaultValue=0.093, group=RadioPaca Tech Art, parameterName=ScaleY_VisorPaca}
   )) {name=ScaleUVs_Visor, nodeColor=<rgba 1 0.7031 0 1>}
  NamedRerouteDeclaration(None=AppendVector(
    A=ScalarParameter() {defaultValue=1, group=RadioPaca Tech Art, parameterName=OffsetX_VisorPaca}
    B=ScalarParameter() {defaultValue=1, group=RadioPaca Tech Art, parameterName=OffsetY_VisorPaca}
   )) {name=OffsetUVs_Visor, nodeColor=<rgba 0.3828 0 1 1>}
  n82 = TextureCoordinate() {coordinateIndex=1}
  n83 = MaterialFunctionCall(Multiplier=Constant() {r=4}) {materialFunction=&Pi}
  n84 = Divide(A=Subtract(A=Rotator(Coordinate=@n82, Time=@n83.Result), B=NamedRerouteUsage()), B=NamedRerouteUsage())
  NamedRerouteDeclaration(None=MaterialFunctionCall(UVs=NamedRerouteDeclaration(None=@n84) {name=ScaledUVs_Visor, nodeColor=<rgba 0 1 0.0625 1>}) {materialFunction=&ScaleUVsByCenter}.0-1 mask) {name=Mask_Visor, nodeColor=<rgba 1 0 0.2578 1>}
  n85 = LinearInterpolate(
   A=LinearInterpolate(A=TextureCoordinate(), B=NamedRerouteUsage(), Alpha=NamedRerouteUsage())
   B=NamedRerouteUsage()
   Alpha=NamedRerouteUsage()
  )
  n86 = NamedRerouteUsage()
  n87 = NamedRerouteUsage()
  n88 = LinearInterpolate(
   A=LinearInterpolate(A=LinearInterpolate(A=@n85, B=@n86, Alpha=@n87), B=NamedRerouteUsage(), Alpha=NamedRerouteUsage())
   B=NamedRerouteUsage()
   Alpha=NamedRerouteUsage()
  )
  n89 = NamedRerouteUsage()
  n90 = NamedRerouteUsage()
  NamedRerouteDeclaration(None=LinearInterpolate(A=LinearInterpolate(A=@n88, B=@n89, Alpha=@n90), B=NamedRerouteUsage(), Alpha=NamedRerouteUsage())) {name=FinalLerpedUVs, nodeColor=<rgba 0 0.5781 1 1>}
  n91 = LinearInterpolate(A=LinearInterpolate(A=Constant2Vector(), B=NamedRerouteUsage(), Alpha=NamedRerouteUsage()), B=NamedRerouteUsage(), Alpha=NamedRerouteUsage())
  n92 = NamedRerouteUsage()
  n93 = NamedRerouteUsage()
  n94 = LinearInterpolate(
   A=LinearInterpolate(A=LinearInterpolate(A=@n91, B=@n92, Alpha=@n93), B=NamedRerouteUsage(), Alpha=NamedRerouteUsage())
   B=NamedRerouteUsage()
   Alpha=NamedRerouteUsage()
  )
  n95 = NamedRerouteUsage()
  n96 = NamedRerouteUsage()
  NamedRerouteDeclaration(None=LinearInterpolate(A=LinearInterpolate(A=@n94, B=@n95, Alpha=@n96), B=NamedRerouteUsage(), Alpha=NamedRerouteUsage())) {name=FinalLerpedGraphBands, nodeColor=<rgba 1 0.8315 0 1>}
  NamedRerouteDeclaration(None=MaterialFunctionCall(X=Constant() {r=12}, Y=Constant() {r=11}) {materialFunction=&MakeFloat2}.Result) {name=Bands_FrontOfDJDeck, nodeColor=<rgba 0.7813 0 1 1>}
  NamedRerouteDeclaration(None=AppendVector(A=Constant() {r=4}, B=Constant() {r=6})) {name=Bands_topOfDJDeck, nodeColor=<rgba 0 1 0.4609 1>}
  NamedRerouteDeclaration(None=AppendVector(A=Constant() {r=8}, B=Constant() {r=17})) {name=Bands_SideOfPacaA, nodeColor=<rgba 1 0.1406 0 1>}
  NamedRerouteDeclaration(None=AppendVector(A=Constant() {r=8}, B=Constant() {r=17})) {name=Bands_SideOfPacaB, nodeColor=<rgba 0 1 0.0391 1>}
  NamedRerouteDeclaration(None=AppendVector(A=Constant() {r=3}, B=Constant() {r=17})) {name=Bands_FrontOfPaca, nodeColor=<rgba 1 0 0.2812 1>}
  NamedRerouteDeclaration(None=AppendVector(A=Constant() {r=3}, B=Constant() {r=17})) {name=Bands_BackOfPaca, nodeColor=<rgba 0 1 0.9766 1>}
  NamedRerouteDeclaration(None=AppendVector(A=Constant() {r=7}, B=Constant() {r=7})) {name=Bands_Visor, nodeColor=<rgba 1 0.6563 0 1>}
  NamedRerouteDeclaration(None=MaterialFunctionCall!(
    In_UV=NamedRerouteUsage()
    In_JustBars=NamedRerouteUsage()
    In_Bands=NamedRerouteUsage()
    In_LowColor=NamedRerouteUsage()
    In_MidColor=NamedRerouteUsage()
    In_HighColor=NamedRerouteUsage()
    In_Spectrum=NamedRerouteUsage()
    In_Amplitude=NamedRerouteUsage()
    In_TimeStamp=NamedRerouteUsage()
   ) {materialFunction=&/MF_2DLEDSpectrum}.MaskedColor) {name=FinalMaskedColorBands, nodeColor=<rgba 0.9688 0 1 1>}
  n97 = Add(A=NamedRerouteUsage(), B=NamedRerouteUsage())
  n98 = NamedRerouteUsage()
  NamedRerouteDeclaration(None=Add(A=Add(A=@n97, B=@n98), B=NamedRerouteUsage())) {name=JustBars, nodeColor=<rgba 0 0.5781 1 1>}
 ]}
