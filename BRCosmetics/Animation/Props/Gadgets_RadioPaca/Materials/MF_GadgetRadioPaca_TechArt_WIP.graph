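# Work-in-progress material function that adds audio-reactive LED spectrum displays to the Radio Paca gadget prop's material attributes.
# Several MF_2DLEDSpectrum instances draw meters with their own colors, band counts and segment counts, with spectrum and amplitude inputs supporting MF_AudioReactivity_FrontendLoop data.
# Parameter-driven offsets, rotations and piecewise range remaps shape each display's UVs.
# Tech-art mask inputs and time-based pulses help blend the results into the incoming attributes.
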MaterialFunction /BRCosmetics/Animation/Props/Gadgets_RadioPaca/Materials/MF_GadgetRadioPaca_TechArt_WIP
graph {expressions {
  MaterialExpressionAbs_0 {class=MaterialExpressionAbs, inputs [None {from=MaterialExpressionConstantBiasScale_0, output=""}], properties {}}
  MaterialExpressionAbs_1 {class=MaterialExpressionAbs, inputs [None {from=MaterialExpressionSubtract_1, output=""}], properties {}}
  MaterialExpressionAdd_0 {
   class=MaterialExpressionAdd
   inputs [
    A {from=MaterialExpressionVectorParameter_1, output=RGB}
    B {from=MaterialExpressionSaturate_0}
   ]
  }
  MaterialExpressionAdd_1 {
   class=MaterialExpressionAdd
   inputs [
    A {from=MaterialExpressionAdd_0}
    B {from=MaterialExpressionSaturate_2}
   ]
  }
  MaterialExpressionAdd_10 {
   class=MaterialExpressionAdd
   inputs [
    A {from=MaterialExpressionMultiply_8}
    B {from=MaterialExpressionReroute_2}
   ]
  }
  MaterialExpressionAdd_11 {
   class=MaterialExpressionAdd
   inputs [
    A {from=MaterialExpressionMultiply_12}
    B {from=MaterialExpressionReroute_15}
   ]
   properties {constA=0.5, constB=0.5}
  }
  MaterialExpressionAdd_12 {
   class=MaterialExpressionAdd
   inputs [
    A {from=MaterialExpressionRotator_10}
    B {from=MaterialExpressionNamedRerouteUsage_22}
   ]
  }
  MaterialExpressionAdd_13 {
   class=MaterialExpressionAdd
   inputs [
    A {from=MaterialExpressionRotator_12}
    B {from=MaterialExpressionNamedRerouteUsage_32}
   ]
  }
  MaterialExpressionAdd_14 {
   class=MaterialExpressionAdd
   inputs [
    A {from=MaterialExpressionMaterialFunctionCall_30, output=G}
    B {from=MaterialExpressionComponentMask_8}
   ]
  }
  MaterialExpressionAdd_15 {
   class=MaterialExpressionAdd
   inputs [
    A {from=MaterialExpressionMultiply_16}
    B {from=MaterialExpressionReroute_11}
   ]
   properties {constA=0.5, constB=0.5}
  }
  MaterialExpressionAdd_16 {
   class=MaterialExpressionAdd
   inputs [
    A {from=MaterialExpressionMaterialFunctionCall_28, output=G}
    B {from=MaterialExpressionComponentMask_6}
   ]
  }
  MaterialExpressionAdd_17 {
   class=MaterialExpressionAdd
   inputs [
    A {from=MaterialExpressionMultiply_17}
    B {from=MaterialExpressionReroute_7}
   ]
   properties {constA=0.5, constB=0.5}
  }
  MaterialExpressionAdd_18 {
   class=MaterialExpressionAdd
   inputs [
    A {from=MaterialExpressionRotator_11}
    B {from=MaterialExpressionNamedRerouteUsage_29}
   ]
  }
  MaterialExpressionAdd_19 {
   class=MaterialExpressionAdd
   inputs [
    A {from=MaterialExpressionMultiply_18}
    B {from=MaterialExpressionReroute_20}
   ]
   properties {constA=0.5, constB=0.5}
  }
  MaterialExpressionAdd_2 {
   class=MaterialExpressionAdd
   inputs [
    A {from=MaterialExpressionLinearInterpolate_1}
    B {from=MaterialExpressionNamedRerouteUsage_24}
   ]
  }
  MaterialExpressionAdd_20 {
   class=MaterialExpressionAdd
   inputs [
    A {from=MaterialExpressionMaterialFunctionCall_25, output=G}
    B {from=MaterialExpressionComponentMask_7}
   ]
  }
  MaterialExpressionAdd_21 {
   class=MaterialExpressionAdd
   inputs [
    A {from=MaterialExpressionRotator_13}
    B {from=MaterialExpressionNamedRerouteUsage_49}
   ]
  }
  MaterialExpressionAdd_22 {
   class=MaterialExpressionAdd
   inputs [
    A {from=MaterialExpressionMaterialFunctionCall_41, output=G}
    B {from=MaterialExpressionComponentMask_9}
   ]
  }
  MaterialExpressionAdd_3 {
   class=MaterialExpressionAdd
   inputs [
    A {from=MaterialExpressionAdd_1}
    B {from=MaterialExpressionSaturate_3}
   ]
  }
  MaterialExpressionAdd_4 {
   class=MaterialExpressionAdd
   inputs [
    A {from=MaterialExpressionAdd_3}
    B {from=MaterialExpressionSaturate_4}
   ]
  }
  MaterialExpressionAdd_5 {
   class=MaterialExpressionAdd
   inputs [
    A {from=MaterialExpressionAdd_4}
    B {from=MaterialExpressionSaturate_5}
   ]
  }
  MaterialExpressionAdd_6 {
   class=MaterialExpressionAdd
   inputs [
    A {from=MaterialExpressionAdd_5}
    B {from=MaterialExpressionSaturate_6}
   ]
  }
  MaterialExpressionAdd_7 {
   class=MaterialExpressionAdd
   inputs [
    A {from=MaterialExpressionAdd_6}
    B {from=MaterialExpressionSaturate_7}
   ]
  }
  MaterialExpressionAdd_8 {
   class=MaterialExpressionAdd
   inputs [
    A {from=MaterialExpressionAdd_2}
    B {from=MaterialExpressionLinearInterpolate_2}
   ]
  }
  MaterialExpressionAdd_9 {
   class=MaterialExpressionAdd
   inputs [
    A {from=MaterialExpressionAdd_6}
    B {from=MaterialExpressionSaturate_8}
   ]
  }
  MaterialExpressionAppendVector_0 {
   class=MaterialExpressionAppendVector
   inputs [
    A {from=MaterialExpressionScalarParameter_0, output=""}
    B {from=MaterialExpressionScalarParameter_1, output=""}
   ]
  }
  MaterialExpressionAppendVector_1 {
   class=MaterialExpressionAppendVector
   inputs [
    A {from=MaterialExpressionScalarParameter_2, output=""}
    B {from=MaterialExpressionScalarParameter_3, output=""}
   ]
  }
  MaterialExpressionAppendVector_10 {
   class=MaterialExpressionAppendVector
   inputs [
    A {from=MaterialExpressionScalarParameter_20, output=""}
    B {from=MaterialExpressionScalarParameter_21, output=""}
   ]
  }
  MaterialExpressionAppendVector_11 {
   class=MaterialExpressionAppendVector
   inputs [
    A {from=MaterialExpressionScalarParameter_22, output=""}
    B {from=MaterialExpressionScalarParameter_23, output=""}
   ]
  }
  MaterialExpressionAppendVector_12 {
   class=MaterialExpressionAppendVector
   inputs [
    A {from=MaterialExpressionScalarParameter_24, output=""}
    B {from=MaterialExpressionScalarParameter_25, output=""}
   ]
  }
  MaterialExpressionAppendVector_13 {
   class=MaterialExpressionAppendVector
   inputs [
    A {from=MaterialExpressionScalarParameter_26, output=""}
    B {from=MaterialExpressionScalarParameter_27, output=""}
   ]
  }
  MaterialExpressionAppendVector_19 {
   class=MaterialExpressionAppendVector
   inputs [
    A {from=MaterialExpressionScalarParameter_43, output=""}
    B {from=MaterialExpressionScalarParameter_44, output=""}
   ]
  }
  MaterialExpressionAppendVector_2 {
   class=MaterialExpressionAppendVector
   inputs [
    A {from=MaterialExpressionScalarParameter_4, output=""}
    B {from=MaterialExpressionScalarParameter_5, output=""}
   ]
  }
  MaterialExpressionAppendVector_20 {
   class=MaterialExpressionAppendVector
   inputs [
    A {from=MaterialExpressionScalarParameter_45, output=""}
    B {from=MaterialExpressionScalarParameter_46, output=""}
   ]
  }
  MaterialExpressionAppendVector_21 {
   class=MaterialExpressionAppendVector
   inputs [
    A {from=MaterialExpressionLinearInterpolate_4, output=""}
    B {from=MaterialExpressionConstant_34, output=""}
   ]
  }
  MaterialExpressionAppendVector_22 {
   class=MaterialExpressionAppendVector
   inputs [
    A {from=MaterialExpressionMultiply_10, output=""}
    B {from=MaterialExpressionReroute_4, output=""}
   ]
  }
  MaterialExpressionAppendVector_23 {
   class=MaterialExpressionAppendVector
   inputs [
    A {from=MaterialExpressionScalarParameter_47, output=""}
    B {from=MaterialExpressionScalarParameter_54, output=""}
   ]
  }
  MaterialExpressionAppendVector_24 {
   class=MaterialExpressionAppendVector
   inputs [
    A {from=MaterialExpressionScalarParameter_55, output=""}
    B {from=MaterialExpressionScalarParameter_64, output=""}
   ]
  }
  MaterialExpressionAppendVector_25 {
   class=MaterialExpressionAppendVector
   inputs [
    A {from=MaterialExpressionMultiply_11, output=""}
    B {from=MaterialExpressionReroute_14, output=""}
   ]
  }
  MaterialExpressionAppendVector_26 {
   class=MaterialExpressionAppendVector
   inputs [
    A {from=MaterialExpressionMultiply_13, output=""}
    B {from=MaterialExpressionReroute_6, output=""}
   ]
  }
  MaterialExpressionAppendVector_28 {
   class=MaterialExpressionAppendVector
   inputs [
    A {from=MaterialExpressionScalarParameter_73, output=""}
    B {from=MaterialExpressionScalarParameter_74, output=""}
   ]
  }
  MaterialExpressionAppendVector_29 {
   class=MaterialExpressionAppendVector
   inputs [
    A {from=MaterialExpressionMultiply_14, output=""}
    B {from=MaterialExpressionReroute_19, output=""}
   ]
  }
  MaterialExpressionAppendVector_3 {
   class=MaterialExpressionAppendVector
   inputs [
    A {from=MaterialExpressionScalarParameter_6, output=""}
    B {from=MaterialExpressionScalarParameter_7, output=""}
   ]
  }
  MaterialExpressionAppendVector_4 {
   class=MaterialExpressionAppendVector
   inputs [
    A {from=MaterialExpressionScalarParameter_8, output=""}
    B {from=MaterialExpressionScalarParameter_9, output=""}
   ]
  }
  MaterialExpressionAppendVector_5 {
   class=MaterialExpressionAppendVector
   inputs [
    A {from=MaterialExpressionScalarParameter_10, output=""}
    B {from=MaterialExpressionScalarParameter_11, output=""}
   ]
  }
  MaterialExpressionAppendVector_6 {
   class=MaterialExpressionAppendVector
   inputs [
    A {from=MaterialExpressionScalarParameter_12, output=""}
    B {from=MaterialExpressionScalarParameter_13, output=""}
   ]
  }
  MaterialExpressionAppendVector_7 {
   class=MaterialExpressionAppendVector
   inputs [
    A {from=MaterialExpressionScalarParameter_14, output=""}
    B {from=MaterialExpressionScalarParameter_15, output=""}
   ]
  }
  MaterialExpressionComponentMask_0 {class=MaterialExpressionComponentMask, inputs [None {from=MaterialExpressionNamedRerouteUsage_0}], properties {g=true}}
  MaterialExpressionComponentMask_1 {class=MaterialExpressionComponentMask, inputs [None {from=MaterialExpressionNamedRerouteUsage_2}], properties {g=true}}
  MaterialExpressionComponentMask_3 {class=MaterialExpressionComponentMask, inputs [None {from=MaterialExpressionNamedRerouteUsage_16}], properties {b=true}}
  MaterialExpressionComponentMask_4 {class=MaterialExpressionComponentMask, inputs [None {from=MaterialExpressionNamedRerouteUsage_23}], properties {b=true}}
  MaterialExpressionComponentMask_5 {class=MaterialExpressionComponentMask, inputs [None {from=MaterialExpressionSubtract_0}], properties {g=true}}
  MaterialExpressionComponentMask_6 {class=MaterialExpressionComponentMask, inputs [None {from=MaterialExpressionSubtract_8}], properties {g=true}}
  MaterialExpressionComponentMask_7 {class=MaterialExpressionComponentMask, inputs [None {from=MaterialExpressionSubtract_4}], properties {g=true}}
  MaterialExpressionComponentMask_8 {class=MaterialExpressionComponentMask, inputs [None {from=MaterialExpressionSubtract_5}], properties {g=true}}
  MaterialExpressionComponentMask_9 {class=MaterialExpressionComponentMask, inputs [None {from=MaterialExpressionSubtract_9}], properties {g=true}}
  MaterialExpressionConstant2Vector_0 {class=MaterialExpressionConstant2Vector, properties {r=0.5}}
  MaterialExpressionConstant2Vector_1 {class=MaterialExpressionConstant2Vector, properties {g=0.5, r=0.5}}
  MaterialExpressionConstant2Vector_10 {class=MaterialExpressionConstant2Vector, properties {g=1, r=1}}
  MaterialExpressionConstant2Vector_11 {class=MaterialExpressionConstant2Vector}
  MaterialExpressionConstant2Vector_12 {class=MaterialExpressionConstant2Vector, properties {g=1, r=1}}
  MaterialExpressionConstant2Vector_13 {class=MaterialExpressionConstant2Vector}
  MaterialExpressionConstant2Vector_2 {class=MaterialExpressionConstant2Vector, properties {g=1, r=1}}
  MaterialExpressionConstant2Vector_3 {class=MaterialExpressionConstant2Vector}
  MaterialExpressionConstant2Vector_4 {class=MaterialExpressionConstant2Vector, properties {g=1, r=1}}
  MaterialExpressionConstant2Vector_5 {class=MaterialExpressionConstant2Vector, properties {g=0.5, r=0.5}}
  MaterialExpressionConstant2Vector_6 {class=MaterialExpressionConstant2Vector, properties {g=1, r=1}}
  MaterialExpressionConstant2Vector_7 {class=MaterialExpressionConstant2Vector}
  MaterialExpressionConstant2Vector_8 {class=MaterialExpressionConstant2Vector, properties {g=0.5, r=0.5}}
  MaterialExpressionConstant2Vector_9 {class=MaterialExpressionConstant2Vector, properties {g=0.5, r=0.5}}
  MaterialExpressionConstantBiasScale_0 {class=MaterialExpressionConstantBiasScale, inputs [None {from=MaterialExpressionFrac_0}], properties {bias=0, scale=2}}
  MaterialExpressionConstant_0 {class=MaterialExpressionConstant, properties {r=11}}
  MaterialExpressionConstant_1 {class=MaterialExpressionConstant, properties {r=12}}
  MaterialExpressionConstant_11 {class=MaterialExpressionConstant, properties {r=6}}
  MaterialExpressionConstant_12 {class=MaterialExpressionConstant, properties {r=17}}
  MaterialExpressionConstant_13 {class=MaterialExpressionConstant, properties {r=8}}
  MaterialExpressionConstant_15 {class=MaterialExpressionConstant, properties {r=17}}
  MaterialExpressionConstant_16 {class=MaterialExpressionConstant, properties {r=3}}
  MaterialExpressionConstant_18 {class=MaterialExpressionConstant, properties {r=17}}
  MaterialExpressionConstant_19 {class=MaterialExpressionConstant, properties {r=3}}
  MaterialExpressionConstant_2 {class=MaterialExpressionConstant, properties {r=6}}
  MaterialExpressionConstant_21 {class=MaterialExpressionConstant, properties {r=12}}
  MaterialExpressionConstant_22 {class=MaterialExpressionConstant, properties {r=1}}
  MaterialExpressionConstant_23 {class=MaterialExpressionConstant, properties {r=6}}
  MaterialExpressionConstant_25 {class=MaterialExpressionConstant}
  MaterialExpressionConstant_26 {class=MaterialExpressionConstant}
  MaterialExpressionConstant_3 {class=MaterialExpressionConstant, properties {r=4}}
  MaterialExpressionConstant_30 {class=MaterialExpressionConstant, properties {r=7}}
  MaterialExpressionConstant_31 {class=MaterialExpressionConstant, properties {r=7}}
  MaterialExpressionConstant_32 {class=MaterialExpressionConstant, properties {r=4}}
  MaterialExpressionConstant_33 {class=MaterialExpressionConstant, properties {r=4}}
  MaterialExpressionConstant_34 {class=MaterialExpressionConstant, properties {r=1}}
  MaterialExpressionConstant_35 {class=MaterialExpressionConstant, properties {r=4}}
  MaterialExpressionConstant_39 {class=MaterialExpressionConstant, properties {r=-0.5}}
  MaterialExpressionConstant_4 {class=MaterialExpressionConstant, properties {r=2}}
  MaterialExpressionConstant_40 {class=MaterialExpressionConstant, properties {r=4}}
  MaterialExpressionConstant_41 {class=MaterialExpressionConstant, properties {r=-0.5}}
  MaterialExpressionConstant_42 {class=MaterialExpressionConstant}
  MaterialExpressionConstant_43 {class=MaterialExpressionConstant, properties {r=-0.5}}
  MaterialExpressionConstant_44 {class=MaterialExpressionConstant}
  MaterialExpressionConstant_45 {class=MaterialExpressionConstant, properties {r=-0.5}}
  MaterialExpressionConstant_46 {class=MaterialExpressionConstant}
  MaterialExpressionConstant_47 {class=MaterialExpressionConstant, properties {r=0.5}}
  MaterialExpressionConstant_48 {class=MaterialExpressionConstant}
  MaterialExpressionConstant_49 {class=MaterialExpressionConstant, properties {r=0.5}}
  MaterialExpressionConstant_5 {class=MaterialExpressionConstant}
  MaterialExpressionConstant_50 {class=MaterialExpressionConstant, properties {r=0.5}}
  MaterialExpressionConstant_51 {class=MaterialExpressionConstant, properties {r=-0.5}}
  MaterialExpressionConstant_52 {class=MaterialExpressionConstant}
  MaterialExpressionConstant_53 {class=MaterialExpressionConstant, properties {r=-0.5}}
  MaterialExpressionConstant_54 {class=MaterialExpressionConstant}
  MaterialExpressionConstant_55 {class=MaterialExpressionConstant, properties {r=0.5}}
  MaterialExpressionConstant_56 {class=MaterialExpressionConstant, properties {r=0.5}}
  MaterialExpressionConstant_57 {class=MaterialExpressionConstant, properties {r=0.5}}
  MaterialExpressionConstant_58 {class=MaterialExpressionConstant, properties {r=4}}
  MaterialExpressionConstant_59 {class=MaterialExpressionConstant, properties {r=-0.5}}
  MaterialExpressionConstant_6 {class=MaterialExpressionConstant, properties {r=1}}
  MaterialExpressionConstant_60 {class=MaterialExpressionConstant}
  MaterialExpressionConstant_61 {class=MaterialExpressionConstant, properties {r=-0.5}}
  MaterialExpressionConstant_62 {class=MaterialExpressionConstant}
  MaterialExpressionConstant_63 {class=MaterialExpressionConstant, properties {r=0.5}}
  MaterialExpressionConstant_64 {class=MaterialExpressionConstant, properties {r=0.5}}
  MaterialExpressionConstant_7 {class=MaterialExpressionConstant, properties {r=17}}
  MaterialExpressionConstant_8 {class=MaterialExpressionConstant, properties {r=8}}
  MaterialExpressionConstant_9 {class=MaterialExpressionConstant}
  MaterialExpressionConvert_0 {
   class=MaterialExpressionConvert
   inputs [
    None {from=MaterialExpressionConstant_3}
    None {from=MaterialExpressionConstant_2}
   ]
   properties {
    nodeName=Make Float2
    convertInputs [
     {
      type=Scalar
      expressionInput {
       expression=&self:MaterialExpressionConstant_3
       inputName=None
       mask=0
       maskA=0
       maskB=0
       maskG=0
       maskR=0
       outputIndex=0
      }
     }
     {
      type=Scalar
      expressionInput {
       expression=&self:MaterialExpressionConstant_2
       inputName=None
       mask=0
       maskA=0
       maskB=0
       maskG=0
       maskR=0
       outputIndex=0
      }
     }
    ]
    convertOutputs [{type=Vector2}]
   }
  }
  MaterialExpressionConvert_1 {
   class=MaterialExpressionConvert
   inputs [
    None {from=MaterialExpressionConstant_1}
    None {from=MaterialExpressionConstant_0}
   ]
   properties {
    nodeName=Make Float2
    convertInputs [
     {
      type=Scalar
      expressionInput {
       expression=&self:MaterialExpressionConstant_1
       inputName=None
       mask=0
       maskA=0
       maskB=0
       maskG=0
       maskR=0
       outputIndex=0
      }
     }
     {
      type=Scalar
      expressionInput {
       expression=&self:MaterialExpressionConstant_0
       inputName=None
       mask=0
       maskA=0
       maskB=0
       maskG=0
       maskR=0
       outputIndex=0
      }
     }
    ]
    convertOutputs [{type=Vector2}]
   }
  }
  MaterialExpressionDivide_0 {
   class=MaterialExpressionDivide
   inputs [
    A {from=MaterialExpressionScalarParameter_28}
    B {from=""}
   ]
   properties {constB=60}
  }
  MaterialExpressionDivide_1 {
   class=MaterialExpressionDivide
   inputs [
    A {from=MaterialExpressionSubtract_11}
    B {from=MaterialExpressionAppendVector_2}
   ]
  }
  MaterialExpressionDivide_2 {
   class=MaterialExpressionDivide
   inputs [
    A {from=MaterialExpressionSubtract_10}
    B {from=MaterialExpressionAppendVector_0}
   ]
  }
  MaterialExpressionDivide_4 {
   class=MaterialExpressionDivide
   inputs [
    A {from=MaterialExpressionTime_9}
    B {from=MaterialExpressionScalarParameter_42}
   ]
  }
  MaterialExpressionFrac_0 {class=MaterialExpressionFrac, inputs [None {from=MaterialExpressionNamedRerouteDeclaration_20, output=""}], properties {}}
  MaterialExpressionFrac_1 {class=MaterialExpressionFrac, inputs [None {from=MaterialExpressionDivide_4, output=""}], properties {}}
  MaterialExpressionFrac_2 {class=MaterialExpressionFrac, inputs [None {from=MaterialExpressionAbs_1, output=""}], properties {}}
  MaterialExpressionFunctionInput_0 {
   class=MaterialExpressionFunctionInput
   inputs [Preview {from="", output=""}]
   properties {inputName=In_MaterialAttributes, inputType=FunctionInput_MaterialAttributes}
  }
  MaterialExpressionFunctionInput_1 {
   class=MaterialExpressionFunctionInput
   inputs [Preview {from=MaterialExpressionTextureSample_0, output=RGB}]
   properties {bUsePreviewValueAsDefault=true, inputName=In_TechArtMask}
  }
  MaterialExpressionFunctionInput_3 {
   class=MaterialExpressionFunctionInput
   inputs [Preview {from=MaterialExpressionTextureSample_2, output=RGB}]
   properties {bUsePreviewValueAsDefault=true, inputName=In_TechArtMask_1}
  }
  MaterialExpressionFunctionOutput_0 {class=MaterialExpressionFunctionOutput, inputs [None {from=MaterialExpressionSetMaterialAttributes_0, output=""}]}
  MaterialExpressionGetMaterialAttributes_0 {
   class=MaterialExpressionGetMaterialAttributes
   inputs [None {from=MaterialExpressionNamedRerouteUsage_1, output=""}]
   properties {attributeGetTypes [#1, #2]}
  }
  MaterialExpressionIf_0 {
   class=MaterialExpressionIf
   inputs [
    A {from=MaterialExpressionNamedRerouteUsage_42, output=""}
    B {from=MaterialExpressionScalarParameter_70, output=""}
    "A > B" {from=MaterialExpressionNamedRerouteUsage_41, output=""}
    "A == B" {from=MaterialExpressionNamedRerouteUsage_40, output=""}
    "A < B" {from=MaterialExpressionNamedRerouteUsage_40, output=""}
   ]
   properties {}
  }
  MaterialExpressionIf_2 {
   class=MaterialExpressionIf
   inputs [
    A {from=MaterialExpressionNamedRerouteUsage_55, output=""}
    B {from=MaterialExpressionScalarParameter_80, output=""}
    "A > B" {from=MaterialExpressionNamedRerouteUsage_54, output=""}
    "A == B" {from=MaterialExpressionNamedRerouteUsage_53, output=""}
    "A < B" {from=MaterialExpressionNamedRerouteUsage_53, output=""}
   ]
   properties {}
  }
  MaterialExpressionIf_3 {
   class=MaterialExpressionIf
   inputs [
    A {from=MaterialExpressionNamedRerouteUsage_36, output=""}
    B {from=MaterialExpressionScalarParameter_57, output=""}
    "A > B" {from=MaterialExpressionNamedRerouteUsage_35, output=""}
    "A == B" {from=MaterialExpressionNamedRerouteUsage_34, output=""}
    "A < B" {from=MaterialExpressionNamedRerouteUsage_34, output=""}
   ]
   properties {}
  }
  MaterialExpressionIf_4 {
   class=MaterialExpressionIf
   inputs [
    A {from=MaterialExpressionNamedRerouteUsage_44, output=""}
    B {from=MaterialExpressionScalarParameter_63, output=""}
    "A > B" {from=MaterialExpressionNamedRerouteUsage_45, output=""}
    "A == B" {from=MaterialExpressionNamedRerouteUsage_46, output=""}
    "A < B" {from=MaterialExpressionNamedRerouteUsage_46, output=""}
   ]
   properties {}
  }
  MaterialExpressionLinearInterpolate_0 {
   class=MaterialExpressionLinearInterpolate
   inputs [
    A {from=MaterialExpressionLinearInterpolate_3}
    B {from=MaterialExpressionNamedRerouteUsage_10}
    Alpha {from=MaterialExpressionComponentMask_0}
   ]
  }
  MaterialExpressionLinearInterpolate_1 {
   class=MaterialExpressionLinearInterpolate
   inputs [
    A {from=MaterialExpressionConstant_26}
    B {from=MaterialExpressionMultiply_6}
    Alpha {from=MaterialExpressionComponentMask_1}
   ]
  }
  MaterialExpressionLinearInterpolate_2 {
   class=MaterialExpressionLinearInterpolate
   inputs [
    A {from=MaterialExpressionConstant_25}
    B {from=MaterialExpressionNamedRerouteUsage_17}
    Alpha {from=MaterialExpressionNamedRerouteUsage_19}
   ]
  }
  MaterialExpressionLinearInterpolate_3 {
   class=MaterialExpressionLinearInterpolate
   inputs [
    A {from=MaterialExpressionNamedRerouteUsage_4}
    B {from=""}
    Alpha {from=MaterialExpressionNamedRerouteUsage_20}
   ]
   properties {constB=0}
  }
  MaterialExpressionLinearInterpolate_4 {
   class=MaterialExpressionLinearInterpolate
   inputs [
    A {from=""}
    B {from=""}
    Alpha {from=MaterialExpressionPower_0}
   ]
   properties {constA=1, constB=2}
  }
  MaterialExpressionLinearInterpolate_5 {
   class=MaterialExpressionLinearInterpolate
   inputs [
    A {from=MaterialExpressionTextureCoordinate_2}
    B {from=MaterialExpressionNamedRerouteUsage_63}
    Alpha {from=MaterialExpressionNamedRerouteUsage_26}
   ]
  }
  MaterialExpressionLinearInterpolate_6 {
   class=MaterialExpressionLinearInterpolate
   inputs [
    A {from=MaterialExpressionLinearInterpolate_5}
    B {from=MaterialExpressionNamedRerouteUsage_102}
    Alpha {from=MaterialExpressionNamedRerouteUsage_103}
   ]
  }
  MaterialExpressionLinearInterpolate_7 {
   class=MaterialExpressionLinearInterpolate
   inputs [
    A {from=MaterialExpressionConvert_0}
    B {from=MaterialExpressionConvert_1}
    Alpha {from=MaterialExpressionNamedRerouteUsage_105}
   ]
  }
  MaterialExpressionLinearInterpolate_8 {
   class=MaterialExpressionLinearInterpolate
   inputs [
    A {from=MaterialExpressionLinearInterpolate_7}
    B {from=""}
    Alpha {from=MaterialExpressionNamedRerouteUsage_104}
   ]
  }
  MaterialExpressionMaterialFunctionCall_0 {
   class=MaterialExpressionMaterialFunctionCall
   inputs [
    UVs {from=MaterialExpressionNamedRerouteDeclaration_15}
    Texture Scale {from=""}
   ]
   properties {materialFunction=&ScaleUVsByCenter}
  }
  MaterialExpressionMaterialFunctionCall_1 {
   class=MaterialExpressionMaterialFunctionCall
   inputs [
    UVs {from=MaterialExpressionDivide_1}
    Texture Scale {from=""}
   ]
   properties {materialFunction=&ScaleUVsByCenter}
  }
  MaterialExpressionMaterialFunctionCall_12 {
   class=MaterialExpressionMaterialFunctionCall
   inputs [Multiplier {from=MaterialExpressionConstant_33}]
   properties {materialFunction=&Pi}
  }
  MaterialExpressionMaterialFunctionCall_15 {
   class=MaterialExpressionMaterialFunctionCall
   inputs [Multiplier {from=MaterialExpressionConstant_23}]
   properties {materialFunction=&Pi}
  }
  MaterialExpressionMaterialFunctionCall_2 {
   class=MaterialExpressionMaterialFunctionCall
   inputs [Multiplier {from=MaterialExpressionConstant_11}]
   properties {materialFunction=&Pi}
  }
  MaterialExpressionMaterialFunctionCall_22 {
   class=MaterialExpressionMaterialFunctionCall
   inputs [
    TimestampA {from=MaterialExpressionScalarParameter_38}
    TransitionLength_In {from=MaterialExpressionScalarParameter_39}
    TimestampB {from=""}
    TransitionLength_Out {from=""}
    Time {from=""}
    Transition In Only? {from=""}
   ]
   properties {materialFunction=&MF_Timestamp_InOut_01}
  }
  MaterialExpressionMaterialFunctionCall_23 {
   class=MaterialExpressionMaterialFunctionCall
   inputs [
    Time {from=""}
    Divisor {from=MaterialExpressionScalarParameter_40}
    Mod {from=MaterialExpressionScalarParameter_41}
   ]
   properties {materialFunction=/Game/Effects/Fort_Effects/Materials/_Functions/MF_ModuloTime}
  }
  MaterialExpressionMaterialFunctionCall_24 {
   class=MaterialExpressionMaterialFunctionCall
   inputs [Multiplier {from=MaterialExpressionConstant_32}]
   properties {materialFunction=&Pi}
  }
  MaterialExpressionMaterialFunctionCall_25 {
   class=MaterialExpressionMaterialFunctionCall
   inputs [Float2 {from=MaterialExpressionSubtract_2}]
   properties {materialFunction=&BreakOutFloat2Components}
  }
  MaterialExpressionMaterialFunctionCall_27 {
   class=MaterialExpressionMaterialFunctionCall
   inputs [Multiplier {from=MaterialExpressionConstant_35}]
   properties {materialFunction=&Pi}
  }
  MaterialExpressionMaterialFunctionCall_28 {
   class=MaterialExpressionMaterialFunctionCall
   inputs [Float2 {from=MaterialExpressionSubtract_7}]
   properties {materialFunction=&BreakOutFloat2Components}
  }
  MaterialExpressionMaterialFunctionCall_29 {
   class=MaterialExpressionMaterialFunctionCall
   inputs [Multiplier {from=MaterialExpressionConstant_40}]
   properties {materialFunction=&Pi}
  }
  MaterialExpressionMaterialFunctionCall_3 {
   class=MaterialExpressionMaterialFunctionCall
   inputs [Multiplier {from=MaterialExpressionConstant_4}]
   properties {materialFunction=&Pi}
  }
  MaterialExpressionMaterialFunctionCall_30 {
   class=MaterialExpressionMaterialFunctionCall
   inputs [Float2 {from=MaterialExpressionSubtract_3}]
   properties {materialFunction=&BreakOutFloat2Components}
  }
  MaterialExpressionMaterialFunctionCall_31 {
   class=MaterialExpressionMaterialFunctionCall
   inputs [
    1 {from=MaterialExpressionScalarParameter_48}
    2 {from=MaterialExpressionScalarParameter_49}
    3 {from=MaterialExpressionScalarParameter_52}
    4 {from=""}
    A {from=MaterialExpressionReroute_9}
   ]
   properties {materialFunction=/Engine/Functions/Engine_MaterialFunctions02/Lerp_Multiple_Float}
  }
  MaterialExpressionMaterialFunctionCall_32 {
   class=MaterialExpressionMaterialFunctionCall
   inputs [
    Input {from=MaterialExpressionNamedRerouteUsage_28}
    Input Low {from=MaterialExpressionConstant_39}
    Input High {from=MaterialExpressionConstant_44}
    Target Low {from=MaterialExpressionConstant_45}
    Target High {from=MaterialExpressionScalarParameter_53}
   ]
   properties {materialFunction=&RemapValueRange}
  }
  MaterialExpressionMaterialFunctionCall_33 {
   class=MaterialExpressionMaterialFunctionCall
   inputs [
    Input {from=MaterialExpressionNamedRerouteUsage_33}
    Input Low {from=MaterialExpressionConstant_48}
    Input High {from=MaterialExpressionConstant_49}
    Target Low {from=MaterialExpressionScalarParameter_56}
    Target High {from=MaterialExpressionConstant_50}
   ]
   properties {materialFunction=&RemapValueRange}
  }
  MaterialExpressionMaterialFunctionCall_34 {
   class=MaterialExpressionMaterialFunctionCall
   inputs [
    Input {from=MaterialExpressionNamedRerouteUsage_48}
    Input Low {from=MaterialExpressionConstant_51}
    Input High {from=MaterialExpressionConstant_52}
    Target Low {from=MaterialExpressionConstant_53}
    Target High {from=MaterialExpressionScalarParameter_58}
   ]
   properties {materialFunction=&RemapValueRange}
  }
  MaterialExpressionMaterialFunctionCall_35 {
   class=MaterialExpressionMaterialFunctionCall
   inputs [
    Input {from=MaterialExpressionNamedRerouteUsage_47}
    Input Low {from=MaterialExpressionConstant_54}
    Input High {from=MaterialExpressionConstant_55}
    Target Low {from=MaterialExpressionScalarParameter_59}
    Target High {from=MaterialExpressionConstant_56}
   ]
   properties {materialFunction=&RemapValueRange}
  }
  MaterialExpressionMaterialFunctionCall_36 {
   class=MaterialExpressionMaterialFunctionCall
   inputs [
    1 {from=MaterialExpressionScalarParameter_60}
    2 {from=MaterialExpressionScalarParameter_61}
    3 {from=MaterialExpressionScalarParameter_62}
    4 {from=""}
    A {from=MaterialExpressionReroute_12}
   ]
   properties {materialFunction=&Lerp_Multiple_Float}
  }
  MaterialExpressionMaterialFunctionCall_37 {
   class=MaterialExpressionMaterialFunctionCall
   inputs [
    1 {from=MaterialExpressionScalarParameter_65}
    2 {from=MaterialExpressionScalarParameter_66}
    3 {from=MaterialExpressionScalarParameter_67}
    4 {from=""}
    A {from=MaterialExpressionReroute_17}
   ]
   properties {materialFunction=&Lerp_Multiple_Float}
  }
  MaterialExpressionMaterialFunctionCall_38 {
   class=MaterialExpressionMaterialFunctionCall
   inputs [
    Input {from=MaterialExpressionNamedRerouteUsage_38}
    Input Low {from=MaterialExpressionConstant_41}
    Input High {from=MaterialExpressionConstant_42}
    Target Low {from=MaterialExpressionConstant_43}
    Target High {from=MaterialExpressionScalarParameter_68}
   ]
   properties {materialFunction=&RemapValueRange}
  }
  MaterialExpressionMaterialFunctionCall_39 {
   class=MaterialExpressionMaterialFunctionCall
   inputs [
    Input {from=MaterialExpressionNamedRerouteUsage_39}
    Input Low {from=MaterialExpressionConstant_46}
    Input High {from=MaterialExpressionConstant_47}
    Target Low {from=MaterialExpressionScalarParameter_69}
    Target High {from=MaterialExpressionConstant_57}
   ]
   properties {materialFunction=&RemapValueRange}
  }
  MaterialExpressionMaterialFunctionCall_4 {
   class=MaterialExpressionMaterialFunctionCall
   inputs [
    In_UV {from=MaterialExpressionLinearInterpolate_6}
    In {from=""}
    In_Bands {from=""}
    In_Segments {from=""}
    In_LowColor {from=""}
    In_MidColor {from=""}
    In_HighColor {from=""}
    In_GraphSize {from=""}
    In_GraphOffset {from=""}
    In_Spectrum {from=""}
    In_Amplitude {from=""}
    In_TimeStamp {from=""}
   ]
   properties {materialFunction=&/MF_2DLEDSpectrum}
  }
  MaterialExpressionMaterialFunctionCall_40 {
   class=MaterialExpressionMaterialFunctionCall
   inputs [Multiplier {from=MaterialExpressionConstant_58}]
   properties {materialFunction=&Pi}
  }
  MaterialExpressionMaterialFunctionCall_41 {
   class=MaterialExpressionMaterialFunctionCall
   inputs [Float2 {from=MaterialExpressionSubtract_6}]
   properties {materialFunction=&BreakOutFloat2Components}
  }
  MaterialExpressionMaterialFunctionCall_42 {
   class=MaterialExpressionMaterialFunctionCall
   inputs [
    1 {from=MaterialExpressionScalarParameter_75}
    2 {from=MaterialExpressionScalarParameter_76}
    3 {from=MaterialExpressionScalarParameter_77}
    4 {from=""}
    A {from=MaterialExpressionReroute_22}
   ]
   properties {materialFunction=&Lerp_Multiple_Float}
  }
  MaterialExpressionMaterialFunctionCall_43 {
   class=MaterialExpressionMaterialFunctionCall
   inputs [
    Input {from=MaterialExpressionNamedRerouteUsage_51}
    Input Low {from=MaterialExpressionConstant_59}
    Input High {from=MaterialExpressionConstant_60}
    Target Low {from=MaterialExpressionConstant_61}
    Target High {from=MaterialExpressionScalarParameter_78}
   ]
   properties {materialFunction=&RemapValueRange}
  }
  MaterialExpressionMaterialFunctionCall_44 {
   class=MaterialExpressionMaterialFunctionCall
   inputs [
    Input {from=MaterialExpressionNamedRerouteUsage_52}
    Input Low {from=MaterialExpressionConstant_62}
    Input High {from=MaterialExpressionConstant_63}
    Target Low {from=MaterialExpressionScalarParameter_79}
    Target High {from=MaterialExpressionConstant_64}
   ]
   properties {materialFunction=&RemapValueRange}
  }
  MaterialExpressionMaterialFunctionCall_48 {
   class=MaterialExpressionMaterialFunctionCall
   inputs [
    InAmplitude {from=""}
    InSpectrum {from=""}
    CustomFakeSpectrum {from=""}
    CustomFakeAmplitude {from=""}
   ]
   properties {materialFunction=/BRCosmetics/Global/MaterialFunctions/AudioReactivity/MF_AudioReactivity_FrontendLoop}
  }
  MaterialExpressionMaterialFunctionCall_49 {
   class=MaterialExpressionMaterialFunctionCall
   inputs [
    In_UV {from=MaterialExpressionNamedRerouteUsage_21}
    In {from=""}
    In_Bands {from=MaterialExpressionConstant_8}
    In_Segments {from=MaterialExpressionConstant_7}
    In_LowColor {from=MaterialExpressionNamedRerouteUsage_61}
    In_MidColor {from=MaterialExpressionNamedRerouteUsage_62}
    In_HighColor {from=MaterialExpressionNamedRerouteUsage_89}
    In_GraphSize {from=MaterialExpressionConstant2Vector_2}
    In_GraphOffset {from=MaterialExpressionConstant2Vector_3}
    In_Spectrum {from=MaterialExpressionNamedRerouteUsage_67}
    In_Amplitude {from=MaterialExpressionNamedRerouteUsage_66}
    In_TimeStamp {from=MaterialExpressionNamedRerouteUsage_65}
   ]
   properties {materialFunction=&MF_2DLEDSpectrum}
  }
  MaterialExpressionMaterialFunctionCall_50 {
   class=MaterialExpressionMaterialFunctionCall
   inputs [
    In_UV {from=MaterialExpressionRotator_0}
    In {from=""}
    In_Bands {from=MaterialExpressionConstant_1}
    In_Segments {from=MaterialExpressionConstant_0}
    In_LowColor {from=MaterialExpressionVectorParameter_2, output=RGB}
    In_MidColor {from=MaterialExpressionVectorParameter_3, output=RGB}
    In_HighColor {from=MaterialExpressionVectorParameter_4, output=RGB}
    In_GraphSize {from=MaterialExpressionAppendVector_0}
    In_GraphOffset {from=MaterialExpressionAppendVector_1}
    In_Spectrum {from=MaterialExpressionNamedRerouteUsage_70}
    In_Amplitude {from=MaterialExpressionNamedRerouteUsage_69}
    In_TimeStamp {from=MaterialExpressionNamedRerouteUsage_68}
   ]
   properties {materialFunction=&MF_2DLEDSpectrum}
  }
  MaterialExpressionMaterialFunctionCall_51 {
   class=MaterialExpressionMaterialFunctionCall
   inputs [
    In_UV {from=MaterialExpressionLinearInterpolate_6}
    In {from=""}
    In_Bands {from=MaterialExpressionLinearInterpolate_7}
    In_Segments {from=MaterialExpressionConstant_2}
    In_LowColor {from=MaterialExpressionNamedRerouteUsage_57}
    In_MidColor {from=MaterialExpressionNamedRerouteUsage_58}
    In_HighColor {from=MaterialExpressionNamedRerouteUsage_60}
    In_GraphSize {from=MaterialExpressionConstant_6}
    In_GraphOffset {from=MaterialExpressionConstant_9}
    In_Spectrum {from=MaterialExpressionNamedRerouteUsage_73}
    In_Amplitude {from=MaterialExpressionNamedRerouteUsage_72}
    In_TimeStamp {from=MaterialExpressionNamedRerouteUsage_71}
   ]
   properties {materialFunction=&MF_2DLEDSpectrum}
  }
  MaterialExpressionMaterialFunctionCall_52 {
   class=MaterialExpressionMaterialFunctionCall
   inputs [
    In_UV {from=MaterialExpressionNamedRerouteUsage_31}
    In {from=""}
    In_Bands {from=MaterialExpressionConstant_13}
    In_Segments {from=MaterialExpressionConstant_12}
    In_LowColor {from=MaterialExpressionNamedRerouteUsage_90}
    In_MidColor {from=MaterialExpressionNamedRerouteUsage_91}
    In_HighColor {from=MaterialExpressionNamedRerouteUsage_92}
    In_GraphSize {from=MaterialExpressionConstant2Vector_10}
    In_GraphOffset {from=MaterialExpressionConstant2Vector_11}
    In_Spectrum {from=MaterialExpressionNamedRerouteUsage_76}
    In_Amplitude {from=MaterialExpressionNamedRerouteUsage_75}
    In_TimeStamp {from=MaterialExpressionNamedRerouteUsage_74}
   ]
   properties {materialFunction=&MF_2DLEDSpectrum}
  }
  MaterialExpressionMaterialFunctionCall_53 {
   class=MaterialExpressionMaterialFunctionCall
   inputs [
    In_UV {from=MaterialExpressionNamedRerouteUsage_43}
    In {from=""}
    In_Bands {from=MaterialExpressionConstant_16}
    In_Segments {from=MaterialExpressionConstant_15}
    In_LowColor {from=MaterialExpressionNamedRerouteUsage_93}
    In_MidColor {from=MaterialExpressionNamedRerouteUsage_94}
    In_HighColor {from=MaterialExpressionNamedRerouteUsage_95}
    In_GraphSize {from=MaterialExpressionConstant2Vector_4}
    In_GraphOffset {from=MaterialExpressionConstant2Vector_7}
    In_Spectrum {from=MaterialExpressionNamedRerouteUsage_79}
    In_Amplitude {from=MaterialExpressionNamedRerouteUsage_78}
    In_TimeStamp {from=MaterialExpressionNamedRerouteUsage_77}
   ]
   properties {materialFunction=&MF_2DLEDSpectrum}
  }
  MaterialExpressionMaterialFunctionCall_54 {
   class=MaterialExpressionMaterialFunctionCall
   inputs [
    In_UV {from=MaterialExpressionNamedRerouteUsage_56}
    In {from=""}
    In_Bands {from=MaterialExpressionConstant_19}
    In_Segments {from=MaterialExpressionConstant_18}
    In_LowColor {from=MaterialExpressionNamedRerouteUsage_99}
    In_MidColor {from=MaterialExpressionNamedRerouteUsage_100}
    In_HighColor {from=MaterialExpressionNamedRerouteUsage_101}
    In_GraphSize {from=MaterialExpressionConstant2Vector_12}
    In_GraphOffset {from=MaterialExpressionConstant2Vector_13}
    In_Spectrum {from=MaterialExpressionNamedRerouteUsage_82}
    In_Amplitude {from=MaterialExpressionNamedRerouteUsage_81}
    In_TimeStamp {from=MaterialExpressionNamedRerouteUsage_80}
   ]
   properties {materialFunction=&MF_2DLEDSpectrum}
  }
  MaterialExpressionMaterialFunctionCall_55 {
   class=MaterialExpressionMaterialFunctionCall
   inputs [
    In_UV {from=MaterialExpressionRotator_6}
    In {from=""}
    In_Bands {from=MaterialExpressionConstant_22}
    In_Segments {from=MaterialExpressionConstant_21}
    In_LowColor {from=MaterialExpressionVectorParameter_20, output=RGB}
    In_MidColor {from=MaterialExpressionVectorParameter_21, output=RGB}
    In_HighColor {from=MaterialExpressionVectorParameter_22, output=RGB}
    In_GraphSize {from=MaterialExpressionAppendVector_12}
    In_GraphOffset {from=MaterialExpressionAppendVector_13}
    In_Spectrum {from=MaterialExpressionNamedRerouteUsage_85}
    In_Amplitude {from=MaterialExpressionNamedRerouteUsage_84}
    In_TimeStamp {from=MaterialExpressionNamedRerouteUsage_83}
   ]
   properties {materialFunction=&MF_2DLEDSpectrum}
  }
  MaterialExpressionMaterialFunctionCall_56 {
   class=MaterialExpressionMaterialFunctionCall
   inputs [
    In_UV {from=MaterialExpressionRotator_9}
    In {from=""}
    In_Bands {from=MaterialExpressionConstant_31}
    In_Segments {from=MaterialExpressionConstant_30}
    In_LowColor {from=MaterialExpressionNamedRerouteUsage_96}
    In_MidColor {from=MaterialExpressionNamedRerouteUsage_97}
    In_HighColor {from=MaterialExpressionNamedRerouteUsage_98}
    In_GraphSize {from=MaterialExpressionAppendVector_19}
    In_GraphOffset {from=MaterialExpressionAppendVector_20}
    In_Spectrum {from=MaterialExpressionNamedRerouteUsage_88}
    In_Amplitude {from=MaterialExpressionNamedRerouteUsage_87}
    In_TimeStamp {from=MaterialExpressionNamedRerouteUsage_86}
   ]
   properties {materialFunction=&MF_2DLEDSpectrum}
  }
  MaterialExpressionMultiply_0 {
   class=MaterialExpressionMultiply
   inputs [
    A {from=MaterialExpressionNamedRerouteDeclaration_15}
    B {from=MaterialExpressionNamedRerouteDeclaration_43}
   ]
  }
  MaterialExpressionMultiply_1 {
   class=MaterialExpressionMultiply
   inputs [
    A {from=MaterialExpressionComponentMask_3}
    B {from=MaterialExpressionSaturate_1}
   ]
  }
  MaterialExpressionMultiply_10 {
   class=MaterialExpressionMultiply
   inputs [
    A {from=MaterialExpressionMaterialFunctionCall_25, output=R}
    B {from=MaterialExpressionMaterialFunctionCall_36, output=Lerp 3 Inputs}
   ]
  }
  MaterialExpressionMultiply_11 {
   class=MaterialExpressionMultiply
   inputs [
    A {from=MaterialExpressionMaterialFunctionCall_30, output=R}
    B {from=MaterialExpressionMaterialFunctionCall_37, output=Lerp 3 Inputs}
   ]
  }
  MaterialExpressionMultiply_12 {
   class=MaterialExpressionMultiply
   inputs [
    A {from=MaterialExpressionAppendVector_25}
    B {from=MaterialExpressionAppendVector_23}
   ]
   properties {constB=2}
  }
  MaterialExpressionMultiply_13 {
   class=MaterialExpressionMultiply
   inputs [
    A {from=MaterialExpressionMaterialFunctionCall_28, output=R}
    B {from=MaterialExpressionMaterialFunctionCall_31, output=Lerp 3 Inputs}
   ]
  }
  MaterialExpressionMultiply_14 {
   class=MaterialExpressionMultiply
   inputs [
    A {from=MaterialExpressionMaterialFunctionCall_41, output=R}
    B {from=MaterialExpressionMaterialFunctionCall_42, output=Lerp 3 Inputs}
   ]
  }
  MaterialExpressionMultiply_16 {
   class=MaterialExpressionMultiply
   inputs [
    A {from=MaterialExpressionAppendVector_22}
    B {from=MaterialExpressionAppendVector_4}
   ]
   properties {constB=2}
  }
  MaterialExpressionMultiply_17 {
   class=MaterialExpressionMultiply
   inputs [
    A {from=MaterialExpressionAppendVector_26}
    B {from=MaterialExpressionAppendVector_6}
   ]
   properties {constB=2}
  }
  MaterialExpressionMultiply_18 {
   class=MaterialExpressionMultiply
   inputs [
    A {from=MaterialExpressionAppendVector_29}
    B {from=MaterialExpressionAppendVector_10}
   ]
   properties {constB=2}
  }
  MaterialExpressionMultiply_19 {
   class=MaterialExpressionMultiply
   inputs [
    A {from=MaterialExpressionMaterialFunctionCall_51, output=Color}
    B {from=""}
   ]
  }
  MaterialExpressionMultiply_2 {
   class=MaterialExpressionMultiply
   inputs [
    A {from=MaterialExpressionTime_0}
    B {from=MaterialExpressionDivide_0}
   ]
  }
  MaterialExpressionMultiply_3 {
   class=MaterialExpressionMultiply
   inputs [
    A {from=MaterialExpressionMultiply_1}
    B {from=""}
   ]
  }
  MaterialExpressionMultiply_4 {
   class=MaterialExpressionMultiply
   inputs [
    A {from=MaterialExpressionNamedRerouteDeclaration_13}
    B {from=MaterialExpressionNamedRerouteDeclaration_13}
   ]
  }
  MaterialExpressionMultiply_5 {
   class=MaterialExpressionMultiply
   inputs [
    A {from=MaterialExpressionVectorParameter_0, output=RGB}
    B {from=MaterialExpressionComponentMask_4}
   ]
  }
  MaterialExpressionMultiply_6 {
   class=MaterialExpressionMultiply
   inputs [
    A {from=MaterialExpressionAdd_9}
    B {from=""}
   ]
  }
  MaterialExpressionMultiply_7 {
   class=MaterialExpressionMultiply
   inputs [
    A {from=MaterialExpressionComponentMask_5}
    B {from=""}
   ]
   properties {constB=2}
  }
  MaterialExpressionMultiply_8 {
   class=MaterialExpressionMultiply
   inputs [
    A {from=MaterialExpressionReroute_0}
    B {from=MaterialExpressionAppendVector_21}
   ]
  }
  MaterialExpressionMultiply_9 {
   class=MaterialExpressionMultiply
   inputs [
    A {from=MaterialExpressionTextureCoordinate_8}
    B {from=MaterialExpressionConstant2Vector_6}
   ]
   properties {constB=2}
  }
  MaterialExpressionNamedRerouteDeclaration_0 {
   class=MaterialExpressionNamedRerouteDeclaration
   inputs [None {from=MaterialExpressionFunctionInput_1}]
   properties {name=In_TechArtMask, nodeColor=<rgba 0.4531 0 1 1>}
  }
  MaterialExpressionNamedRerouteDeclaration_1 {
   class=MaterialExpressionNamedRerouteDeclaration
   inputs [None {from=MaterialExpressionFunctionInput_0}]
   properties {name=In_MaterialAttributes, nodeColor=<rgba 0 1 0.1328 1>}
  }
  MaterialExpressionNamedRerouteDeclaration_10 {
   class=MaterialExpressionNamedRerouteDeclaration
   inputs [None {from=MaterialExpressionMaterialFunctionCall_53, output=MaskedColor}]
   properties {name=MaskedColor_FrontofPaca, nodeColor=<rgba 0.3143 0.2824 1 1>}
  }
  MaterialExpressionNamedRerouteDeclaration_11 {
   class=MaterialExpressionNamedRerouteDeclaration
   inputs [None {from=MaterialExpressionMaterialFunctionCall_54, output=MaskedColor}]
   properties {name=MaskedColor_BackofPaca, nodeColor=<rgba 0.088 1 0.9427 1>}
  }
  MaterialExpressionNamedRerouteDeclaration_12 {
   class=MaterialExpressionNamedRerouteDeclaration
   inputs [None {from=MaterialExpressionMaterialFunctionCall_55, output=MaskedColor}]
   properties {name=MaskedColor_SingleBar, nodeColor=<rgba 0.5654 1 0.1355 1>}
  }
  MaterialExpressionNamedRerouteDeclaration_13 {
   class=MaterialExpressionNamedRerouteDeclaration
   inputs [None {from=MaterialExpressionOneMinus_1}]
   properties {name=DefaultBeat, nodeColor=<rgba 0 1 0.5078 1>}
  }
  MaterialExpressionNamedRerouteDeclaration_14 {
   class=MaterialExpressionNamedRerouteDeclaration
   inputs [None {from=MaterialExpressionMultiply_3}]
   properties {name=StuffTimedToBeat, nodeColor=<rgba 0.5469 0 1 1>}
  }
  MaterialExpressionNamedRerouteDeclaration_15 {
   class=MaterialExpressionNamedRerouteDeclaration
   inputs [None {from=MaterialExpressionDivide_2}]
   properties {name=ScaledUVs_Deck, nodeColor=<rgba 0.0078 1 0 1>}
  }
  MaterialExpressionNamedRerouteDeclaration_16 {
   class=MaterialExpressionNamedRerouteDeclaration
   inputs [None {from=MaterialExpressionVectorParameter_35, output=RGB}]
   properties {name=LowColor_Paca, nodeColor=<rgba 1 0.6328 0 1>}
  }
  MaterialExpressionNamedRerouteDeclaration_17 {
   class=MaterialExpressionNamedRerouteDeclaration
   inputs [None {from=MaterialExpressionVectorParameter_36, output=RGB}]
   properties {name=MidColor_Paca, nodeColor=<rgba 0.3125 0 1 1>}
  }
  MaterialExpressionNamedRerouteDeclaration_18 {
   class=MaterialExpressionNamedRerouteDeclaration
   inputs [None {from=MaterialExpressionStaticSwitchParameter_5}]
   properties {name=TimeStamp, nodeColor=<rgba 0.125 1 0 1>}
  }
  MaterialExpressionNamedRerouteDeclaration_19 {
   class=MaterialExpressionNamedRerouteDeclaration
   inputs [None {from=MaterialExpressionMaterialFunctionCall_56, output=MaskedColor}]
   properties {name=MaskedColor_VisorPaca, nodeColor=<rgba 0.088 1 0.9427 1>}
  }
  MaterialExpressionNamedRerouteDeclaration_2 {
   class=MaterialExpressionNamedRerouteDeclaration
   inputs [None {from=MaterialExpressionGetMaterialAttributes_0, output=EmissiveColor}]
   properties {name=In_Emissive, nodeColor=<rgba 1 0.5156 0 1>}
  }
  MaterialExpressionNamedRerouteDeclaration_20 {
   class=MaterialExpressionNamedRerouteDeclaration
   inputs [None {from=MaterialExpressionMultiply_2}]
   properties {name=DefaultBPM, nodeColor=<rgba 0.0547 1 0 1>}
  }
  MaterialExpressionNamedRerouteDeclaration_21 {
   class=MaterialExpressionNamedRerouteDeclaration
   inputs [None {from=MaterialExpressionFunctionInput_3}]
   properties {name=In_TechArtMask2, nodeColor=<rgba 0.4531 0 1 1>}
  }
  MaterialExpressionNamedRerouteDeclaration_22 {
   class=MaterialExpressionNamedRerouteDeclaration
   inputs [None {from=MaterialExpressionMultiply_5}]
   properties {name=GlowStickEmissiveColor, nodeColor=<rgba 0.1534 1 0.078 1>}
  }
  MaterialExpressionNamedRerouteDeclaration_23 {
   class=MaterialExpressionNamedRerouteDeclaration
   inputs [None {from=MaterialExpressionAdd_12}]
   properties {name=SkewedUVsSideA, nodeColor=<rgba 0.8177 0.5911 0.0894 1>}
  }
  MaterialExpressionNamedRerouteDeclaration_24 {
   class=MaterialExpressionNamedRerouteDeclaration
   inputs [None {from=MaterialExpressionAppendVector_5}]
   properties {name=OffsetSideA, nodeColor=<rgba 1 0.0661 0.0035 1>}
  }
  MaterialExpressionNamedRerouteDeclaration_25 {
   class=MaterialExpressionNamedRerouteDeclaration
   inputs [None {from=MaterialExpressionAdd_18}]
   properties {name=SkewedUVsSideB, nodeColor=<rgba 1 0 0.0938 1>}
  }
  MaterialExpressionNamedRerouteDeclaration_26 {
   class=MaterialExpressionNamedRerouteDeclaration
   inputs [None {from=MaterialExpressionAppendVector_7}]
   properties {name=OffsetSideB, nodeColor=<rgba 0 1 0.9531 1>}
  }
  MaterialExpressionNamedRerouteDeclaration_27 {
   class=MaterialExpressionNamedRerouteDeclaration
   inputs [None {from=MaterialExpressionAdd_16}]
   properties {name=OffsetRangeB, nodeColor=<rgba 1 0 0.4219 1>}
  }
  MaterialExpressionNamedRerouteDeclaration_28 {
   class=MaterialExpressionNamedRerouteDeclaration
   inputs [None {from=MaterialExpressionMaterialFunctionCall_32, output=Result}]
   properties {name=LowerRangeSideB, nodeColor=<rgba 0 0.7422 1 1>}
  }
  MaterialExpressionNamedRerouteDeclaration_29 {
   class=MaterialExpressionNamedRerouteDeclaration
   inputs [None {from=MaterialExpressionMaterialFunctionCall_33, output=Result}]
   properties {name=UpperRangeSideB, nodeColor=<rgba 1 0.9375 0 1>}
  }
  MaterialExpressionNamedRerouteDeclaration_3 {
   class=MaterialExpressionNamedRerouteDeclaration
   inputs [None {from=MaterialExpressionGetMaterialAttributes_0, output=BaseColor}]
   properties {name=In_BaseColor, nodeColor=<rgba 0.1953 0 1 1>}
  }
  MaterialExpressionNamedRerouteDeclaration_30 {
   class=MaterialExpressionNamedRerouteDeclaration
   inputs [None {from=MaterialExpressionMaterialFunctionCall_34, output=Result}]
   properties {name=LowerRangeSideA, nodeColor=<rgba 0 0.7422 1 1>}
  }
  MaterialExpressionNamedRerouteDeclaration_31 {
   class=MaterialExpressionNamedRerouteDeclaration
   inputs [None {from=MaterialExpressionMaterialFunctionCall_35, output=Result}]
   properties {name=UpperRangeSideA, nodeColor=<rgba 1 0.9375 0 1>}
  }
  MaterialExpressionNamedRerouteDeclaration_32 {
   class=MaterialExpressionNamedRerouteDeclaration
   inputs [None {from=MaterialExpressionAdd_20}]
   properties {name=OffsetRangeA, nodeColor=<rgba 1 0.2578 0 1>}
  }
  MaterialExpressionNamedRerouteDeclaration_33 {
   class=MaterialExpressionNamedRerouteDeclaration
   inputs [None {from=MaterialExpressionAdd_13}]
   properties {name=SkewedUVsFront, nodeColor=<rgba 0.0307 0.0284 0.0723 1>}
  }
  MaterialExpressionNamedRerouteDeclaration_34 {
   class=MaterialExpressionNamedRerouteDeclaration
   inputs [None {from=MaterialExpressionAppendVector_24}]
   properties {name=OffsetSideB_1, nodeColor=<rgba 0 1 0.9531 1>}
  }
  MaterialExpressionNamedRerouteDeclaration_35 {
   class=MaterialExpressionNamedRerouteDeclaration
   inputs [None {from=MaterialExpressionAdd_14}]
   properties {name=OffsetRangeB_1, nodeColor=<rgba 1 0 0.4219 1>}
  }
  MaterialExpressionNamedRerouteDeclaration_36 {
   class=MaterialExpressionNamedRerouteDeclaration
   inputs [None {from=MaterialExpressionMaterialFunctionCall_38, output=Result}]
   properties {name=LowerRangeSideB_1, nodeColor=<rgba 0 0.7422 1 1>}
  }
  MaterialExpressionNamedRerouteDeclaration_37 {
   class=MaterialExpressionNamedRerouteDeclaration
   inputs [None {from=MaterialExpressionMaterialFunctionCall_39, output=Result}]
   properties {name=UpperRangeSideB_1, nodeColor=<rgba 1 0.9375 0 1>}
  }
  MaterialExpressionNamedRerouteDeclaration_38 {
   class=MaterialExpressionNamedRerouteDeclaration
   inputs [None {from=MaterialExpressionAdd_21}]
   properties {name=SkewedUVsBack, nodeColor=<rgba 0.159 0.8385 0.7847 1>}
  }
  MaterialExpressionNamedRerouteDeclaration_39 {
   class=MaterialExpressionNamedRerouteDeclaration
   inputs [None {from=MaterialExpressionAppendVector_11}]
   properties {name=OffsetSideB_2, nodeColor=<rgba 0 1 0.9531 1>}
  }
  MaterialExpressionNamedRerouteDeclaration_4 {
   class=MaterialExpressionNamedRerouteDeclaration
   inputs [None {from=MaterialExpressionAdd_8}]
   properties {name=Out_Emissive, nodeColor=<rgba 0.125 1 0 1>}
  }
  MaterialExpressionNamedRerouteDeclaration_40 {
   class=MaterialExpressionNamedRerouteDeclaration
   inputs [None {from=MaterialExpressionAdd_22}]
   properties {name=OffsetRangeB_2, nodeColor=<rgba 1 0 0.4219 1>}
  }
  MaterialExpressionNamedRerouteDeclaration_41 {
   class=MaterialExpressionNamedRerouteDeclaration
   inputs [None {from=MaterialExpressionMaterialFunctionCall_43, output=Result}]
   properties {name=LowerRangeSideB_2, nodeColor=<rgba 0 0.7422 1 1>}
  }
  MaterialExpressionNamedRerouteDeclaration_42 {
   class=MaterialExpressionNamedRerouteDeclaration
   inputs [None {from=MaterialExpressionMaterialFunctionCall_44, output=Result}]
   properties {name=UpperRangeSideB_2, nodeColor=<rgba 1 0.9375 0 1>}
  }
  MaterialExpressionNamedRerouteDeclaration_43 {
   class=MaterialExpressionNamedRerouteDeclaration
   inputs [None {from=MaterialExpressionMaterialFunctionCall_0, output=0-1 mask}]
   properties {name=Mask_FrontOfDJDeck, nodeColor=<rgba 1 0.1406 0 1>}
  }
  MaterialExpressionNamedRerouteDeclaration_44 {
   class=MaterialExpressionNamedRerouteDeclaration
   inputs [None {from=MaterialExpressionVectorParameter_37, output=RGB}]
   properties {name=HighColor_Paca, nodeColor=<rgba 0.0078 1 0 1>}
  }
  MaterialExpressionNamedRerouteDeclaration_45 {
   class=MaterialExpressionNamedRerouteDeclaration
   inputs [None {from=MaterialExpressionMaterialFunctionCall_50, output=Color}]
   properties {name=Color_FrontOfDJDeck, nodeColor=<rgba 0 0.25 1 1>}
  }
  MaterialExpressionNamedRerouteDeclaration_46 {
   class=MaterialExpressionNamedRerouteDeclaration
   inputs [None {from=MaterialExpressionMaterialFunctionCall_1, output=0-1 mask}]
   properties {name=Mask_DeckTop, nodeColor=<rgba 0.6406 1 0 1>}
  }
  MaterialExpressionNamedRerouteDeclaration_47 {
   class=MaterialExpressionNamedRerouteDeclaration
   inputs [None {from=MaterialExpressionDivide_1}]
   properties {name=ScaledUV_DeckTop, nodeColor=<rgba 1 0 0.9609 1>}
  }
  MaterialExpressionNamedRerouteDeclaration_48 {
   class=MaterialExpressionNamedRerouteDeclaration
   inputs [None {from=MaterialExpressionMaterialFunctionCall_48, output=Amplitude}]
   properties {name=Amplitude, nodeColor=<rgba 0 1 0.1094 1>}
  }
  MaterialExpressionNamedRerouteDeclaration_49 {
   class=MaterialExpressionNamedRerouteDeclaration
   inputs [None {from=MaterialExpressionMaterialFunctionCall_48, output=Spectrum}]
   properties {name=Spectrum, nodeColor=<rgba 1 0 0.75 1>}
  }
  MaterialExpressionNamedRerouteDeclaration_5 {
   class=MaterialExpressionNamedRerouteDeclaration
   inputs [None {from=MaterialExpressionLinearInterpolate_0}]
   properties {name=Out_BaseColor, nodeColor=<rgba 1 0 0.4453 1>}
  }
  MaterialExpressionNamedRerouteDeclaration_50 {
   class=MaterialExpressionNamedRerouteDeclaration
   inputs [None {from=MaterialExpressionPower_1}]
   properties {name=AmplitudeAdjusted, nodeColor=<rgba 0 1 0.9531 1>}
  }
  MaterialExpressionNamedRerouteDeclaration_6 {
   class=MaterialExpressionNamedRerouteDeclaration
   inputs [None {from=MaterialExpressionMaterialFunctionCall_50, output=MaskedColor}]
   properties {name=MaskedColor_FrontofDJDeck, nodeColor=<rgba 0 0.7656 1 1>}
  }
  MaterialExpressionNamedRerouteDeclaration_7 {
   class=MaterialExpressionNamedRerouteDeclaration
   inputs [None {from=MaterialExpressionMaterialFunctionCall_51, output=MaskedColor}]
   properties {name=MaskedColor_TopofDJDeck, nodeColor=<rgba 0.1304 1 0.025 1>}
  }
  MaterialExpressionNamedRerouteDeclaration_8 {
   class=MaterialExpressionNamedRerouteDeclaration
   inputs [None {from=MaterialExpressionMaterialFunctionCall_49, output=MaskedColor}]
   properties {name=MaskedColor_SideofPacaA, nodeColor=<rgba 1 0.6541 0.0237 1>}
  }
  MaterialExpressionNamedRerouteDeclaration_9 {
   class=MaterialExpressionNamedRerouteDeclaration
   inputs [None {from=MaterialExpressionMaterialFunctionCall_52, output=MaskedColor}]
   properties {name=MaskedColor_SideofPacaB, nodeColor=<rgba 1 0.0334 0.3227 1>}
  }
  MaterialExpressionNamedRerouteUsage_0 {class=MaterialExpressionNamedRerouteUsage}
  MaterialExpressionNamedRerouteUsage_1 {class=MaterialExpressionNamedRerouteUsage}
  MaterialExpressionNamedRerouteUsage_10 {class=MaterialExpressionNamedRerouteUsage}
  MaterialExpressionNamedRerouteUsage_100 {class=MaterialExpressionNamedRerouteUsage}
  MaterialExpressionNamedRerouteUsage_101 {class=MaterialExpressionNamedRerouteUsage}
  MaterialExpressionNamedRerouteUsage_102 {class=MaterialExpressionNamedRerouteUsage}
  MaterialExpressionNamedRerouteUsage_103 {class=MaterialExpressionNamedRerouteUsage}
  MaterialExpressionNamedRerouteUsage_104 {class=MaterialExpressionNamedRerouteUsage}
  MaterialExpressionNamedRerouteUsage_105 {class=MaterialExpressionNamedRerouteUsage}
  MaterialExpressionNamedRerouteUsage_11 {class=MaterialExpressionNamedRerouteUsage}
  MaterialExpressionNamedRerouteUsage_12 {class=MaterialExpressionNamedRerouteUsage}
  MaterialExpressionNamedRerouteUsage_13 {class=MaterialExpressionNamedRerouteUsage}
  MaterialExpressionNamedRerouteUsage_14 {class=MaterialExpressionNamedRerouteUsage}
  MaterialExpressionNamedRerouteUsage_15 {class=MaterialExpressionNamedRerouteUsage}
  MaterialExpressionNamedRerouteUsage_16 {class=MaterialExpressionNamedRerouteUsage}
  MaterialExpressionNamedRerouteUsage_17 {class=MaterialExpressionNamedRerouteUsage}
  MaterialExpressionNamedRerouteUsage_19 {class=MaterialExpressionNamedRerouteUsage}
  MaterialExpressionNamedRerouteUsage_2 {class=MaterialExpressionNamedRerouteUsage}
  MaterialExpressionNamedRerouteUsage_20 {class=MaterialExpressionNamedRerouteUsage}
  MaterialExpressionNamedRerouteUsage_21 {class=MaterialExpressionNamedRerouteUsage}
  MaterialExpressionNamedRerouteUsage_22 {class=MaterialExpressionNamedRerouteUsage}
  MaterialExpressionNamedRerouteUsage_23 {class=MaterialExpressionNamedRerouteUsage}
  MaterialExpressionNamedRerouteUsage_24 {class=MaterialExpressionNamedRerouteUsage}
  MaterialExpressionNamedRerouteUsage_25 {class=MaterialExpressionNamedRerouteUsage}
  MaterialExpressionNamedRerouteUsage_26 {class=MaterialExpressionNamedRerouteUsage}
  MaterialExpressionNamedRerouteUsage_27 {class=MaterialExpressionNamedRerouteUsage}
  MaterialExpressionNamedRerouteUsage_28 {class=MaterialExpressionNamedRerouteUsage}
  MaterialExpressionNamedRerouteUsage_29 {class=MaterialExpressionNamedRerouteUsage}
  MaterialExpressionNamedRerouteUsage_30 {class=MaterialExpressionNamedRerouteUsage}
  MaterialExpressionNamedRerouteUsage_31 {class=MaterialExpressionNamedRerouteUsage}
  MaterialExpressionNamedRerouteUsage_32 {class=MaterialExpressionNamedRerouteUsage}
  MaterialExpressionNamedRerouteUsage_33 {class=MaterialExpressionNamedRerouteUsage}
  MaterialExpressionNamedRerouteUsage_34 {class=MaterialExpressionNamedRerouteUsage}
  MaterialExpressionNamedRerouteUsage_35 {class=MaterialExpressionNamedRerouteUsage}
  MaterialExpressionNamedRerouteUsage_36 {class=MaterialExpressionNamedRerouteUsage}
  MaterialExpressionNamedRerouteUsage_37 {class=MaterialExpressionNamedRerouteUsage}
  MaterialExpressionNamedRerouteUsage_38 {class=MaterialExpressionNamedRerouteUsage}
  MaterialExpressionNamedRerouteUsage_39 {class=MaterialExpressionNamedRerouteUsage}
  MaterialExpressionNamedRerouteUsage_4 {class=MaterialExpressionNamedRerouteUsage}
  MaterialExpressionNamedRerouteUsage_40 {class=MaterialExpressionNamedRerouteUsage}
  MaterialExpressionNamedRerouteUsage_41 {class=MaterialExpressionNamedRerouteUsage}
  MaterialExpressionNamedRerouteUsage_42 {class=MaterialExpressionNamedRerouteUsage}
  MaterialExpressionNamedRerouteUsage_43 {class=MaterialExpressionNamedRerouteUsage}
  MaterialExpressionNamedRerouteUsage_44 {class=MaterialExpressionNamedRerouteUsage}
  MaterialExpressionNamedRerouteUsage_45 {class=MaterialExpressionNamedRerouteUsage}
  MaterialExpressionNamedRerouteUsage_46 {class=MaterialExpressionNamedRerouteUsage}
  MaterialExpressionNamedRerouteUsage_47 {class=MaterialExpressionNamedRerouteUsage}
  MaterialExpressionNamedRerouteUsage_48 {class=MaterialExpressionNamedRerouteUsage}
  MaterialExpressionNamedRerouteUsage_49 {class=MaterialExpressionNamedRerouteUsage}
  MaterialExpressionNamedRerouteUsage_5 {class=MaterialExpressionNamedRerouteUsage}
  MaterialExpressionNamedRerouteUsage_50 {class=MaterialExpressionNamedRerouteUsage}
  MaterialExpressionNamedRerouteUsage_51 {class=MaterialExpressionNamedRerouteUsage}
  MaterialExpressionNamedRerouteUsage_52 {class=MaterialExpressionNamedRerouteUsage}
  MaterialExpressionNamedRerouteUsage_53 {class=MaterialExpressionNamedRerouteUsage}
  MaterialExpressionNamedRerouteUsage_54 {class=MaterialExpressionNamedRerouteUsage}
  MaterialExpressionNamedRerouteUsage_55 {class=MaterialExpressionNamedRerouteUsage}
  MaterialExpressionNamedRerouteUsage_56 {class=MaterialExpressionNamedRerouteUsage}
  MaterialExpressionNamedRerouteUsage_57 {class=MaterialExpressionNamedRerouteUsage}
  MaterialExpressionNamedRerouteUsage_58 {class=MaterialExpressionNamedRerouteUsage}
  MaterialExpressionNamedRerouteUsage_59 {class=MaterialExpressionNamedRerouteUsage}
  MaterialExpressionNamedRerouteUsage_6 {class=MaterialExpressionNamedRerouteUsage}
  MaterialExpressionNamedRerouteUsage_60 {class=MaterialExpressionNamedRerouteUsage}
  MaterialExpressionNamedRerouteUsage_61 {class=MaterialExpressionNamedRerouteUsage}
  MaterialExpressionNamedRerouteUsage_62 {class=MaterialExpressionNamedRerouteUsage}
  MaterialExpressionNamedRerouteUsage_63 {class=MaterialExpressionNamedRerouteUsage}
  MaterialExpressionNamedRerouteUsage_64 {class=MaterialExpressionNamedRerouteUsage}
  MaterialExpressionNamedRerouteUsage_65 {class=MaterialExpressionNamedRerouteUsage}
  MaterialExpressionNamedRerouteUsage_66 {class=MaterialExpressionNamedRerouteUsage}
  MaterialExpressionNamedRerouteUsage_67 {class=MaterialExpressionNamedRerouteUsage}
  MaterialExpressionNamedRerouteUsage_68 {class=MaterialExpressionNamedRerouteUsage}
  MaterialExpressionNamedRerouteUsage_69 {class=MaterialExpressionNamedRerouteUsage}
  MaterialExpressionNamedRerouteUsage_7 {class=MaterialExpressionNamedRerouteUsage}
  MaterialExpressionNamedRerouteUsage_70 {class=MaterialExpressionNamedRerouteUsage}
  MaterialExpressionNamedRerouteUsage_71 {class=MaterialExpressionNamedRerouteUsage}
  MaterialExpressionNamedRerouteUsage_72 {class=MaterialExpressionNamedRerouteUsage}
  MaterialExpressionNamedRerouteUsage_73 {class=MaterialExpressionNamedRerouteUsage}
  MaterialExpressionNamedRerouteUsage_74 {class=MaterialExpressionNamedRerouteUsage}
  MaterialExpressionNamedRerouteUsage_75 {class=MaterialExpressionNamedRerouteUsage}
  MaterialExpressionNamedRerouteUsage_76 {class=MaterialExpressionNamedRerouteUsage}
  MaterialExpressionNamedRerouteUsage_77 {class=MaterialExpressionNamedRerouteUsage}
  MaterialExpressionNamedRerouteUsage_78 {class=MaterialExpressionNamedRerouteUsage}
  MaterialExpressionNamedRerouteUsage_79 {class=MaterialExpressionNamedRerouteUsage}
  MaterialExpressionNamedRerouteUsage_80 {class=MaterialExpressionNamedRerouteUsage}
  MaterialExpressionNamedRerouteUsage_81 {class=MaterialExpressionNamedRerouteUsage}
  MaterialExpressionNamedRerouteUsage_82 {class=MaterialExpressionNamedRerouteUsage}
  MaterialExpressionNamedRerouteUsage_83 {class=MaterialExpressionNamedRerouteUsage}
  MaterialExpressionNamedRerouteUsage_84 {class=MaterialExpressionNamedRerouteUsage}
  MaterialExpressionNamedRerouteUsage_85 {class=MaterialExpressionNamedRerouteUsage}
  MaterialExpressionNamedRerouteUsage_86 {class=MaterialExpressionNamedRerouteUsage}
  MaterialExpressionNamedRerouteUsage_87 {class=MaterialExpressionNamedRerouteUsage}
  MaterialExpressionNamedRerouteUsage_88 {class=MaterialExpressionNamedRerouteUsage}
  MaterialExpressionNamedRerouteUsage_89 {class=MaterialExpressionNamedRerouteUsage}
  MaterialExpressionNamedRerouteUsage_9 {class=MaterialExpressionNamedRerouteUsage}
  MaterialExpressionNamedRerouteUsage_90 {class=MaterialExpressionNamedRerouteUsage}
  MaterialExpressionNamedRerouteUsage_91 {class=MaterialExpressionNamedRerouteUsage}
  MaterialExpressionNamedRerouteUsage_92 {class=MaterialExpressionNamedRerouteUsage}
  MaterialExpressionNamedRerouteUsage_93 {class=MaterialExpressionNamedRerouteUsage}
  MaterialExpressionNamedRerouteUsage_94 {class=MaterialExpressionNamedRerouteUsage}
  MaterialExpressionNamedRerouteUsage_95 {class=MaterialExpressionNamedRerouteUsage}
  MaterialExpressionNamedRerouteUsage_96 {class=MaterialExpressionNamedRerouteUsage}
  MaterialExpressionNamedRerouteUsage_97 {class=MaterialExpressionNamedRerouteUsage}
  MaterialExpressionNamedRerouteUsage_98 {class=MaterialExpressionNamedRerouteUsage}
  MaterialExpressionNamedRerouteUsage_99 {class=MaterialExpressionNamedRerouteUsage}
  MaterialExpressionOneMinus_1 {class=MaterialExpressionOneMinus, inputs [None {from=MaterialExpressionAbs_0}]}
  MaterialExpressionPanner_1 {
   class=MaterialExpressionPanner
   inputs [
    Coordinate {from=""}
    Time {from=""}
    Speed {from=""}
   ]
  }
  MaterialExpressionPower_0 {
   class=MaterialExpressionPower
   inputs [
    Base {from=MaterialExpressionFrac_2}
    Exp {from=""}
   ]
  }
  MaterialExpressionPower_1 {
   class=MaterialExpressionPower
   inputs [
    Base {from=MaterialExpressionNamedRerouteUsage_64}
    Exp {from=MaterialExpressionScalarParameter_86}
   ]
   properties {constExponent=8}
  }
  MaterialExpressionReroute_0 {class=MaterialExpressionReroute, inputs [None {from=MaterialExpressionReroute_1}]}
  MaterialExpressionReroute_1 {class=MaterialExpressionReroute, inputs [None {from=MaterialExpressionSubtract_0}]}
  MaterialExpressionReroute_10 {class=MaterialExpressionReroute, inputs [None {from=MaterialExpressionConstant2Vector_8}]}
  MaterialExpressionReroute_11 {class=MaterialExpressionReroute, inputs [None {from=MaterialExpressionReroute_10}]}
  MaterialExpressionReroute_12 {class=MaterialExpressionReroute, inputs [None {from=MaterialExpressionIf_4}]}
  MaterialExpressionReroute_13 {class=MaterialExpressionReroute, inputs [None {from=MaterialExpressionMaterialFunctionCall_30, output=G}]}
  MaterialExpressionReroute_14 {class=MaterialExpressionReroute, inputs [None {from=MaterialExpressionReroute_13}]}
  MaterialExpressionReroute_15 {class=MaterialExpressionReroute, inputs [None {from=MaterialExpressionReroute_16}]}
  MaterialExpressionReroute_16 {class=MaterialExpressionReroute, inputs [None {from=MaterialExpressionConstant2Vector_1}]}
  MaterialExpressionReroute_17 {class=MaterialExpressionReroute, inputs [None {from=MaterialExpressionIf_0}]}
  MaterialExpressionReroute_18 {class=MaterialExpressionReroute, inputs [None {from=MaterialExpressionMaterialFunctionCall_41, output=G}]}
  MaterialExpressionReroute_19 {class=MaterialExpressionReroute, inputs [None {from=MaterialExpressionReroute_18}]}
  MaterialExpressionReroute_2 {class=MaterialExpressionReroute, inputs [None {from=MaterialExpressionConstant2Vector_0}]}
  MaterialExpressionReroute_20 {class=MaterialExpressionReroute, inputs [None {from=MaterialExpressionReroute_21}]}
  MaterialExpressionReroute_21 {class=MaterialExpressionReroute, inputs [None {from=MaterialExpressionConstant2Vector_9}]}
  MaterialExpressionReroute_22 {class=MaterialExpressionReroute, inputs [None {from=MaterialExpressionIf_2}]}
  MaterialExpressionReroute_3 {class=MaterialExpressionReroute, inputs [None {from=MaterialExpressionMaterialFunctionCall_25, output=G}]}
  MaterialExpressionReroute_4 {class=MaterialExpressionReroute, inputs [None {from=MaterialExpressionReroute_3}]}
  MaterialExpressionReroute_5 {class=MaterialExpressionReroute, inputs [None {from=MaterialExpressionMaterialFunctionCall_28, output=G}]}
  MaterialExpressionReroute_6 {class=MaterialExpressionReroute, inputs [None {from=MaterialExpressionReroute_5}]}
  MaterialExpressionReroute_7 {class=MaterialExpressionReroute, inputs [None {from=MaterialExpressionReroute_8}]}
  MaterialExpressionReroute_8 {class=MaterialExpressionReroute, inputs [None {from=MaterialExpressionConstant2Vector_5}]}
  MaterialExpressionReroute_9 {class=MaterialExpressionReroute, inputs [None {from=MaterialExpressionIf_3}]}
  MaterialExpressionRotator_0 {
   class=MaterialExpressionRotator
   inputs [
    Coordinate {from=MaterialExpressionTextureCoordinate_0, output=""}
    Time {from=MaterialExpressionMaterialFunctionCall_2, output=Result}
   ]
   properties {}
  }
  MaterialExpressionRotator_1 {
   class=MaterialExpressionRotator
   inputs [
    Coordinate {from=MaterialExpressionTextureCoordinate_1, output=""}
    Time {from=MaterialExpressionConstant_5, output=""}
   ]
   properties {}
  }
  MaterialExpressionRotator_10 {
   class=MaterialExpressionRotator
   inputs [
    Coordinate {from=MaterialExpressionAdd_15, output=""}
    Time {from=MaterialExpressionMaterialFunctionCall_12, output=Result}
   ]
   properties {}
  }
  MaterialExpressionRotator_11 {
   class=MaterialExpressionRotator
   inputs [
    Coordinate {from=MaterialExpressionAdd_17, output=""}
    Time {from=MaterialExpressionMaterialFunctionCall_27, output=Result}
   ]
   properties {}
  }
  MaterialExpressionRotator_12 {
   class=MaterialExpressionRotator
   inputs [
    Coordinate {from=MaterialExpressionAdd_11, output=""}
    Time {from=MaterialExpressionMaterialFunctionCall_29, output=Result}
   ]
   properties {}
  }
  MaterialExpressionRotator_13 {
   class=MaterialExpressionRotator
   inputs [
    Coordinate {from=MaterialExpressionAdd_19, output=""}
    Time {from=MaterialExpressionMaterialFunctionCall_40, output=Result}
   ]
   properties {}
  }
  MaterialExpressionRotator_6 {
   class=MaterialExpressionRotator
   inputs [
    Coordinate {from=MaterialExpressionTextureCoordinate_6, output=""}
    Time {from=MaterialExpressionMaterialFunctionCall_15, output=Result}
   ]
   properties {}
  }
  MaterialExpressionRotator_9 {
   class=MaterialExpressionRotator
   inputs [
    Coordinate {from=MaterialExpressionTextureCoordinate_9, output=""}
    Time {from=MaterialExpressionMaterialFunctionCall_24, output=Result}
   ]
   properties {}
  }
  MaterialExpressionSaturate_0 {class=MaterialExpressionSaturate, inputs [None {from=MaterialExpressionNamedRerouteUsage_59}]}
  MaterialExpressionSaturate_1 {class=MaterialExpressionSaturate, inputs [None {from=MaterialExpressionMultiply_4}]}
  MaterialExpressionSaturate_2 {class=MaterialExpressionSaturate, inputs [None {from=MaterialExpressionNamedRerouteUsage_9}]}
  MaterialExpressionSaturate_3 {class=MaterialExpressionSaturate, inputs [None {from=MaterialExpressionNamedRerouteUsage_11}]}
  MaterialExpressionSaturate_4 {class=MaterialExpressionSaturate, inputs [None {from=MaterialExpressionNamedRerouteUsage_12}]}
  MaterialExpressionSaturate_5 {class=MaterialExpressionSaturate, inputs [None {from=MaterialExpressionNamedRerouteUsage_13}]}
  MaterialExpressionSaturate_6 {class=MaterialExpressionSaturate, inputs [None {from=MaterialExpressionNamedRerouteUsage_14}]}
  MaterialExpressionSaturate_7 {class=MaterialExpressionSaturate, inputs [None {from=MaterialExpressionNamedRerouteUsage_15}]}
  MaterialExpressionSaturate_8 {class=MaterialExpressionSaturate, inputs [None {from=MaterialExpressionNamedRerouteUsage_25}]}
  MaterialExpressionScalarParameter_0 {class=MaterialExpressionScalarParameter, properties {defaultValue=0.404, group=RadioPaca Tech Art, parameterName=ScaleX_FrontofDJDeck}}
  MaterialExpressionScalarParameter_1 {class=MaterialExpressionScalarParameter, properties {defaultValue=0.086, group=RadioPaca Tech Art, parameterName=ScaleY_FrontofDJDeck}}
  MaterialExpressionScalarParameter_10 {class=MaterialExpressionScalarParameter, properties {defaultValue=-0.16, group=RadioPaca Tech Art, parameterName=OffsetX_SideofPacaA}}
  MaterialExpressionScalarParameter_11 {class=MaterialExpressionScalarParameter, properties {group=RadioPaca Tech Art, parameterName=OffsetY_SideofPacaA}}
  MaterialExpressionScalarParameter_12 {class=MaterialExpressionScalarParameter, properties {defaultValue=1, group=RadioPaca Tech Art, parameterName=ScaleX_SideofPacaB}}
  MaterialExpressionScalarParameter_13 {class=MaterialExpressionScalarParameter, properties {defaultValue=1, group=RadioPaca Tech Art, parameterName=ScaleY_SideofPacaB}}
  MaterialExpressionScalarParameter_14 {class=MaterialExpressionScalarParameter, properties {defaultValue=0.587, group=RadioPaca Tech Art, parameterName=OffsetX_SideofPacaB}}
  MaterialExpressionScalarParameter_15 {class=MaterialExpressionScalarParameter, properties {defaultValue=0.563, group=RadioPaca Tech Art, parameterName=OffsetY_SideofPacaB}}
  MaterialExpressionScalarParameter_2 {class=MaterialExpressionScalarParameter, properties {defaultValue=0.52, group=RadioPaca Tech Art, parameterName=OffsetX_FrontofDJDeck}}
  MaterialExpressionScalarParameter_20 {class=MaterialExpressionScalarParameter, properties {group=RadioPaca Tech Art, parameterName=ScaleX_BackofPaca}}
  MaterialExpressionScalarParameter_21 {class=MaterialExpressionScalarParameter, properties {defaultValue=0.093, group=RadioPaca Tech Art, parameterName=ScaleY_BackofPaca}}
  MaterialExpressionScalarParameter_22 {class=MaterialExpressionScalarParameter, properties {defaultValue=1, group=RadioPaca Tech Art, parameterName=OffsetX_BackofPaca}}
  MaterialExpressionScalarParameter_23 {class=MaterialExpressionScalarParameter, properties {defaultValue=1, group=RadioPaca Tech Art, parameterName=OffsetY_BackofPaca}}
  MaterialExpressionScalarParameter_24 {class=MaterialExpressionScalarParameter, properties {defaultValue=0.08, group=RadioPaca Tech Art, parameterName=ScaleX_SingleBar}}
  MaterialExpressionScalarParameter_25 {class=MaterialExpressionScalarParameter, properties {defaultValue=0.093, group=RadioPaca Tech Art, parameterName=ScaleY_SingleBar}}
  MaterialExpressionScalarParameter_26 {class=MaterialExpressionScalarParameter, properties {defaultValue=0.587, group=RadioPaca Tech Art, parameterName=OffsetX_SingleBar}}
  MaterialExpressionScalarParameter_27 {class=MaterialExpressionScalarParameter, properties {defaultValue=0.563, group=RadioPaca Tech Art, parameterName=OffsetY_SingleBar}}
  MaterialExpressionScalarParameter_28 {class=MaterialExpressionScalarParameter, properties {defaultValue=120, parameterName=Deafault BPM}}
  MaterialExpressionScalarParameter_3 {class=MaterialExpressionScalarParameter, properties {defaultValue=0.904, group=RadioPaca Tech Art, parameterName=OffsetY_FrontofDJDeck}}
  MaterialExpressionScalarParameter_38 {class=MaterialExpressionScalarParameter, properties {defaultValue=1, parameterName=Timestamp, sortPriority=30}}
  MaterialExpressionScalarParameter_39 {class=MaterialExpressionScalarParameter, properties {defaultValue=0.1, parameterName=TransitionLength, sortPriority=30}}
  MaterialExpressionScalarParameter_4 {class=MaterialExpressionScalarParameter, properties {defaultValue=0.08, group=RadioPaca Tech Art, parameterName=ScaleX_TopofDJDeck}}
  MaterialExpressionScalarParameter_40 {
   class=MaterialExpressionScalarParameter
   properties {defaultValue=1.5, group=FNCSShield_Screen, parameterName=DebugTimeDivisor, sortPriority=30}
  }
  MaterialExpressionScalarParameter_41 {
   class=MaterialExpressionScalarParameter
   properties {defaultValue=2, group=FNCSShield_Screen, parameterName=DebugTimeModulo, sortPriority=30}
  }
  MaterialExpressionScalarParameter_42 {
   class=MaterialExpressionScalarParameter
   properties {defaultValue=3, group=FNCSShield_Screen, parameterName=DebugTransitionLength, sortPriority=30}
  }
  MaterialExpressionScalarParameter_43 {class=MaterialExpressionScalarParameter, properties {group=RadioPaca Tech Art, parameterName=ScaleX_VisorPaca}}
  MaterialExpressionScalarParameter_44 {class=MaterialExpressionScalarParameter, properties {defaultValue=0.093, group=RadioPaca Tech Art, parameterName=ScaleY_VisorPaca}}
  MaterialExpressionScalarParameter_45 {class=MaterialExpressionScalarParameter, properties {defaultValue=1, group=RadioPaca Tech Art, parameterName=OffsetX_VisorPaca}}
  MaterialExpressionScalarParameter_46 {class=MaterialExpressionScalarParameter, properties {defaultValue=1, group=RadioPaca Tech Art, parameterName=OffsetY_VisorPaca}}
  MaterialExpressionScalarParameter_47 {class=MaterialExpressionScalarParameter, properties {defaultValue=1, group=RadioPaca Tech Art, parameterName=ScaleX_FrontofPacaB}}
  MaterialExpressionScalarParameter_48 {class=MaterialExpressionScalarParameter, properties {defaultValue=0.632, group=RadioPaca Tech Art, parameterName=SideBWidthA}}
  MaterialExpressionScalarParameter_49 {class=MaterialExpressionScalarParameter, properties {defaultValue=0.39, group=RadioPaca Tech Art, parameterName=SideBWidthB}}
  MaterialExpressionScalarParameter_5 {class=MaterialExpressionScalarParameter, properties {defaultValue=0.093, group=RadioPaca Tech Art, parameterName=ScaleY_TopofDJDeck}}
  MaterialExpressionScalarParameter_52 {class=MaterialExpressionScalarParameter, properties {defaultValue=0.39, group=RadioPaca Tech Art, parameterName=SideBWidthC}}
  MaterialExpressionScalarParameter_53 {class=MaterialExpressionScalarParameter, properties {group=RadioPaca Tech Art, parameterName=SideBMidValue}}
  MaterialExpressionScalarParameter_54 {class=MaterialExpressionScalarParameter, properties {defaultValue=1, group=RadioPaca Tech Art, parameterName=ScaleY_FrontofPacaB}}
  MaterialExpressionScalarParameter_55 {class=MaterialExpressionScalarParameter, properties {defaultValue=0.587, group=RadioPaca Tech Art, parameterName=OffsetX_FrontofPacaB}}
  MaterialExpressionScalarParameter_56 {class=MaterialExpressionScalarParameter, properties {group=RadioPaca Tech Art, parameterName=SideBMidValue}}
  MaterialExpressionScalarParameter_57 {class=MaterialExpressionScalarParameter, properties {group=RadioPaca Tech Art, parameterName=SideBMidValue}}
  MaterialExpressionScalarParameter_58 {class=MaterialExpressionScalarParameter, properties {group=RadioPaca Tech Art, parameterName=SideAMidValue}}
  MaterialExpressionScalarParameter_59 {class=MaterialExpressionScalarParameter, properties {group=RadioPaca Tech Art, parameterName=SideAMidValue}}
  MaterialExpressionScalarParameter_6 {class=MaterialExpressionScalarParameter, properties {defaultValue=0.587, group=RadioPaca Tech Art, parameterName=OffsetX_TopofDJDeck}}
  MaterialExpressionScalarParameter_60 {class=MaterialExpressionScalarParameter, properties {defaultValue=0.632, group=RadioPaca Tech Art, parameterName=SideAWidthA}}
  MaterialExpressionScalarParameter_61 {class=MaterialExpressionScalarParameter, properties {defaultValue=0.39, group=RadioPaca Tech Art, parameterName=SideAWidthB}}
  MaterialExpressionScalarParameter_62 {class=MaterialExpressionScalarParameter, properties {defaultValue=0.39, group=RadioPaca Tech Art, parameterName=SideAWidthC}}
  MaterialExpressionScalarParameter_63 {class=MaterialExpressionScalarParameter, properties {group=RadioPaca Tech Art, parameterName=SideAMidValue}}
  MaterialExpressionScalarParameter_64 {class=MaterialExpressionScalarParameter, properties {defaultValue=0.563, group=RadioPaca Tech Art, parameterName=OffsetY_FrontofPacaB}}
  MaterialExpressionScalarParameter_65 {class=MaterialExpressionScalarParameter, properties {defaultValue=0.632, group=RadioPaca Tech Art, parameterName=FrontWidthA}}
  MaterialExpressionScalarParameter_66 {class=MaterialExpressionScalarParameter, properties {defaultValue=0.39, group=RadioPaca Tech Art, parameterName=FrontWidthB}}
  MaterialExpressionScalarParameter_67 {class=MaterialExpressionScalarParameter, properties {defaultValue=0.39, group=RadioPaca Tech Art, parameterName=FrontWidthC}}
  MaterialExpressionScalarParameter_68 {class=MaterialExpressionScalarParameter, properties {group=RadioPaca Tech Art, parameterName=FrontMidValue}}
  MaterialExpressionScalarParameter_69 {class=MaterialExpressionScalarParameter, properties {group=RadioPaca Tech Art, parameterName=FrontMidValue}}
  MaterialExpressionScalarParameter_7 {class=MaterialExpressionScalarParameter, properties {defaultValue=0.563, group=RadioPaca Tech Art, parameterName=OffsetY_TopofDJDeck}}
  MaterialExpressionScalarParameter_70 {class=MaterialExpressionScalarParameter, properties {group=RadioPaca Tech Art, parameterName=FrontMidValue}}
  MaterialExpressionScalarParameter_73 {class=MaterialExpressionScalarParameter, properties {defaultValue=0.587, group=RadioPaca Tech Art, parameterName=OffsetX_FrontofPacaB}}
  MaterialExpressionScalarParameter_74 {class=MaterialExpressionScalarParameter, properties {defaultValue=0.563, group=RadioPaca Tech Art, parameterName=OffsetY_FrontofPacaB}}
  MaterialExpressionScalarParameter_75 {class=MaterialExpressionScalarParameter, properties {defaultValue=0.632, group=RadioPaca Tech Art, parameterName=BackWidthA}}
  MaterialExpressionScalarParameter_76 {class=MaterialExpressionScalarParameter, properties {defaultValue=0.39, group=RadioPaca Tech Art, parameterName=BackWidthB}}
  MaterialExpressionScalarParameter_77 {class=MaterialExpressionScalarParameter, properties {defaultValue=0.39, group=RadioPaca Tech Art, parameterName=BackWidthC}}
  MaterialExpressionScalarParameter_78 {class=MaterialExpressionScalarParameter, properties {group=RadioPaca Tech Art, parameterName=BackMidValue}}
  MaterialExpressionScalarParameter_79 {class=MaterialExpressionScalarParameter, properties {group=RadioPaca Tech Art, parameterName=BackMidValue}}
  MaterialExpressionScalarParameter_8 {class=MaterialExpressionScalarParameter, properties {defaultValue=3.5211, group=RadioPaca Tech Art, parameterName=ScaleX_SideofPacaA}}
  MaterialExpressionScalarParameter_80 {class=MaterialExpressionScalarParameter, properties {group=RadioPaca Tech Art, parameterName=BackMidValue}}
  MaterialExpressionScalarParameter_86 {class=MaterialExpressionScalarParameter, properties {defaultValue=8, group=TechArt, parameterName=AmplitudeEmissiveExp}}
  MaterialExpressionScalarParameter_9 {class=MaterialExpressionScalarParameter, properties {defaultValue=2, group=RadioPaca Tech Art, parameterName=ScaleY_SideofPacaA}}
  MaterialExpressionSetMaterialAttributes_0 {
   class=MaterialExpressionSetMaterialAttributes
   inputs [
    MaterialAttributes {from=MaterialExpressionGetMaterialAttributes_0, output=MaterialAttributes}
    EmissiveColor {from=MaterialExpressionNamedRerouteUsage_6, output=""}
    BaseColor {from=MaterialExpressionNamedRerouteUsage_7, output=""}
   ]
   properties {attributeSetTypes [#1, #2]}
  }
  MaterialExpressionStaticSwitchParameter_5 {
   class=MaterialExpressionStaticSwitchParameter
   inputs [
    True {from=MaterialExpressionFrac_1, output=""}
    False {from=MaterialExpressionVertexInterpolator_1, output=PS}
   ]
   properties {group=FNCSShield_Screen_Graphics, parameterName=DebugModeOn, sortPriority=0}
  }
  MaterialExpressionSubtract_0 {
   class=MaterialExpressionSubtract
   inputs [
    A {from=MaterialExpressionMultiply_9}
    B {from=MaterialExpressionConstant2Vector_0}
   ]
  }
  MaterialExpressionSubtract_1 {
   class=MaterialExpressionSubtract
   inputs [
    A {from=MaterialExpressionMultiply_7}
    B {from=""}
   ]
  }
  MaterialExpressionSubtract_10 {
   class=MaterialExpressionSubtract
   inputs [
    A {from=MaterialExpressionRotator_0}
    B {from=MaterialExpressionAppendVector_1}
   ]
  }
  MaterialExpressionSubtract_11 {
   class=MaterialExpressionSubtract
   inputs [
    A {from=MaterialExpressionRotator_1}
    B {from=MaterialExpressionAppendVector_3}
   ]
  }
  MaterialExpressionSubtract_2 {
   class=MaterialExpressionSubtract
   inputs [
    A {from=MaterialExpressionSubtract_4}
    B {from=MaterialExpressionConstant2Vector_8}
   ]
  }
  MaterialExpressionSubtract_3 {
   class=MaterialExpressionSubtract
   inputs [
    A {from=MaterialExpressionSubtract_5}
    B {from=MaterialExpressionConstant2Vector_1}
   ]
  }
  MaterialExpressionSubtract_4 {
   class=MaterialExpressionSubtract
   inputs [
    A {from=MaterialExpressionTextureCoordinate_10}
    B {from=MaterialExpressionNamedRerouteUsage_27}
   ]
  }
  MaterialExpressionSubtract_5 {
   class=MaterialExpressionSubtract
   inputs [
    A {from=MaterialExpressionTextureCoordinate_12}
    B {from=MaterialExpressionNamedRerouteUsage_37}
   ]
  }
  MaterialExpressionSubtract_6 {
   class=MaterialExpressionSubtract
   inputs [
    A {from=MaterialExpressionSubtract_9}
    B {from=MaterialExpressionConstant2Vector_9}
   ]
  }
  MaterialExpressionSubtract_7 {
   class=MaterialExpressionSubtract
   inputs [
    A {from=MaterialExpressionSubtract_8}
    B {from=MaterialExpressionConstant2Vector_5}
   ]
  }
  MaterialExpressionSubtract_8 {
   class=MaterialExpressionSubtract
   inputs [
    A {from=MaterialExpressionTextureCoordinate_11}
    B {from=MaterialExpressionNamedRerouteUsage_30}
   ]
  }
  MaterialExpressionSubtract_9 {
   class=MaterialExpressionSubtract
   inputs [
    A {from=MaterialExpressionTextureCoordinate_13}
    B {from=MaterialExpressionNamedRerouteUsage_50}
   ]
  }
  MaterialExpressionTextureCoordinate_0 {class=MaterialExpressionTextureCoordinate}
  MaterialExpressionTextureCoordinate_1 {class=MaterialExpressionTextureCoordinate}
  MaterialExpressionTextureCoordinate_10 {class=MaterialExpressionTextureCoordinate}
  MaterialExpressionTextureCoordinate_11 {class=MaterialExpressionTextureCoordinate}
  MaterialExpressionTextureCoordinate_12 {class=MaterialExpressionTextureCoordinate}
  MaterialExpressionTextureCoordinate_13 {class=MaterialExpressionTextureCoordinate}
  MaterialExpressionTextureCoordinate_2 {class=MaterialExpressionTextureCoordinate}
  MaterialExpressionTextureCoordinate_6 {class=MaterialExpressionTextureCoordinate}
  MaterialExpressionTextureCoordinate_8 {class=MaterialExpressionTextureCoordinate}
  MaterialExpressionTextureCoordinate_9 {class=MaterialExpressionTextureCoordinate, properties {coordinateIndex=1}}
  MaterialExpressionTextureSample_0 {
   class=MaterialExpressionTextureSample
   inputs [
    UVs {from=""}
    Tex {from=""}
    Apply View MipBias {from=""}
   ]
   properties {samplerType=SAMPLERTYPE_Masks, texture=/BRCosmetics/Animation/Props/Gadgets_RadioPaca/Textures/T_RadioPaca_Gadgets_FX}
  }
  MaterialExpressionTextureSample_2 {
   class=MaterialExpressionTextureSample
   inputs [
    UVs {from=""}
    Tex {from=""}
    Apply View MipBias {from=""}
   ]
   properties {texture=/BRCosmetics/Animation/Props/Gadgets_RadioPaca/Textures/T_RadioPaca_Gadgets_FX2}
  }
  MaterialExpressionTime_0 {class=MaterialExpressionTime}
  MaterialExpressionTime_9 {class=MaterialExpressionTime}
  MaterialExpressionVectorParameter_0 {class=MaterialExpressionVectorParameter, properties {group=RadioPaca Tech Art, parameterName=GlowStick, defaultValue=<rgba 0.0883 3 0.08 1>}}
  MaterialExpressionVectorParameter_1 {class=MaterialExpressionVectorParameter, properties {group=RadioPaca Tech Art, parameterName=BaseEmissive, defaultValue=<rgba 0 0 0 1>}}
  MaterialExpressionVectorParameter_2 {
   class=MaterialExpressionVectorParameter
   properties {group=RadioPaca Tech Art, parameterName=LowColor_FrontofDJDeck, defaultValue=<rgba 0 1 1 1>}
  }
  MaterialExpressionVectorParameter_20 {
   class=MaterialExpressionVectorParameter
   properties {group=RadioPaca Tech Art, parameterName=LowColor_SingleBar, defaultValue=<rgba 0 1 1 1>}
  }
  MaterialExpressionVectorParameter_21 {
   class=MaterialExpressionVectorParameter
   properties {group=RadioPaca Tech Art, parameterName=MidColor_SingleBar, defaultValue=<rgba 0.0027 0.0027 0.9823 1>}
  }
  MaterialExpressionVectorParameter_22 {
   class=MaterialExpressionVectorParameter
   properties {group=RadioPaca Tech Art, parameterName=HighColor_SingleBar, defaultValue=<rgba 1 0 1 1>}
  }
  MaterialExpressionVectorParameter_3 {
   class=MaterialExpressionVectorParameter
   properties {group=RadioPaca Tech Art, parameterName=MidColor_FrontofDJDeck, defaultValue=<rgba 0 1 1 1>}
  }
  MaterialExpressionVectorParameter_35 {
   class=MaterialExpressionVectorParameter
   properties {group=RadioPaca Tech Art, parameterName=LowColor_Paca, defaultValue=<rgba 0.0561 0.1248 0.1878 1>}
  }
  MaterialExpressionVectorParameter_36 {
   class=MaterialExpressionVectorParameter
   properties {group=RadioPaca Tech Art, parameterName=MidColor_Paca, defaultValue=<rgba 0.0123 0.0865 0.4179 1>}
  }
  MaterialExpressionVectorParameter_37 {
   class=MaterialExpressionVectorParameter
   properties {group=RadioPaca Tech Art, parameterName=HighColor_Paca, defaultValue=<rgba 0.0545 0.2335 0.9304 1>}
  }
  MaterialExpressionVectorParameter_4 {
   class=MaterialExpressionVectorParameter
   properties {group=RadioPaca Tech Art, parameterName=HighColor_FrontofDJDeck, defaultValue=<rgba 0.8469 0.0762 0.8469 1>}
  }
  MaterialExpressionVertexInterpolator_1 {class=MaterialExpressionVertexInterpolator, inputs [VS {from=MaterialExpressionMaterialFunctionCall_22, output=Linear}], properties {}}
 }}
