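# Material function for the RadioPaca gadget prop that sets up audio reactivity data, apparently as an unfinished scaffold.
# It calls the shared MF_AudioReactivity function and exposes Amplitude and Spectrum as named reroutes.
# A separate branch raises a reroute value to a power of one and splits it into four components, which feed nothing.
# The function output has no input connected.
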MaterialFunction /BRCosmetics/Animation/Props/Gadgets_RadioPaca/Materials/MF_Gadget_RadioPaca
graph {nodes [
  FunctionOutput()
  MaterialFunctionCall(Float4=Power(Base=NamedRerouteUsage(), Exp=Constant() {r=1})) {materialFunction=&BreakOutFloat4Components}
  n1 = MaterialFunctionCall() {materialFunction=/Game/Packages/Fortress_SharedMaterials/Base_Material_Functions/BaseCharacterShaderFunctions/Audio/MF_AudioReactivity}
  NamedRerouteDeclaration(None=@n1.Amplitude) {name=Amplitude, nodeColor=<rgba 0 1 0.9062 1>}
  NamedRerouteDeclaration(None=@n1.Spectrum) {name=Spectrum, nodeColor=<rgba 1 0.5859 0 1>}
 ]}
